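# Control Rig for the Kurohomura medium male player body that procedurally corrects secondary bones from the animated pose.
# Elbow, shoulder and clavicle measurements become clamped weights for additive tricep, bicep, floating-clavicle, trapezius and latissimus offsets, alongside transform constraints.
# Thigh swing counter-rotates both belt bases, and dyn_thighsLookAt on the right thigh aims at the left.
# The Crouch Walk bool selects between duplicated chains differing only in right trapezius and latissimus offsets.
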ControlRigBlueprint /BRCosmetics/Characters/Player/Male/Medium/Bodies/M_MED_Kurohomura/Meshes/M_MED_Kurohomura_CtrlRig
graphs [
 RigVMFunctionLibrary {links [], nodes []}
 RigVMModel {nodes [
   n1 = &RigUnit_BeginExecution()
   n2 = &RigVMFunction_MathQuaternionToEuler(RotationOrder=XYZ, Value=&RigUnit_GetRelativeBoneTransform(Bone=lowerarm_r, Space=upperarm_r).Transform.Rotation)
   /Script/ControlRig.RigUnit_ModifyBoneTransforms(
    BoneToModify=((Bone="upperarm_tricep_r",Transform=(Rotation=(X=0,Y=0,Z=0,W=1),Translation=(X=3,Y=2,Z=0),Scale3D=(X=1,Y=1.5,Z=1.5))),(Bone="None",Transform=(Rotation=(X=0,Y=0,Z=0,W=1),Translation=(X=0,Y=0,Z=0),Scale3D=(X=1,Y=1,Z=1))))
    Mode=AdditiveLocal
    WeightMaximum=1
    WeightMinimum=0
    ExecutePin=@n1.ExecutePin
    Weight=&RigVMFunction_MathFloatRemap(SourceMaximum=20, SourceMinimum=120, TargetMinimum=0, bClamp=true, Value=@n2.Result.Z).Result
   ) {is=struct RigVMUnitNode}
   > &RigUnit_ModifyBoneTransforms(
    BoneToModify=((Bone="upperarm_tricep_r",Transform=(Rotation=(X=0,Y=0,Z=0,W=1),Translation=(X=-5,Y=0,Z=0),Scale3D=(X=1.3,Y=0.4,Z=0.4))),(Bone="None",Transform=(Rotation=(X=0,Y=0,Z=0,W=1),Translation=(X=0,Y=0,Z=0),Scale3D=(X=1,Y=1,Z=1))))
    Mode=AdditiveLocal
    WeightMaximum=1
    WeightMinimum=0
    Weight=&RigVMFunction_MathFloatRemap(SourceMaximum=120, SourceMinimum=20, TargetMinimum=0, bClamp=true, Value=@n2.Result.Z).Result
   )
   n3 = &RigVMFunction_MathQuaternionToEuler(RotationOrder=XYZ, Value=&RigUnit_GetRelativeBoneTransform(Bone=lowerarm_l, Space=upperarm_l).Transform.Rotation)
   > &RigUnit_ModifyBoneTransforms(
    BoneToModify=((Bone="upperarm_tricep_l",Transform=(Rotation=(X=0,Y=0,Z=0,W=1),Translation=(X=-3,Y=-2,Z=0),Scale3D=(X=1,Y=1.5,Z=1.5))),(Bone="None",Transform=(Rotation=(X=0,Y=0,Z=0,W=1),Translation=(X=0,Y=0,Z=0),Scale3D=(X=1,Y=1,Z=1))))
    Mode=AdditiveLocal
    WeightMaximum=1
    WeightMinimum=0
    Weight=&RigVMFunction_MathFloatRemap(SourceMaximum=20, SourceMinimum=120, TargetMinimum=0, bClamp=true, Value=@n3.Result.Z).Result
   )
   > &RigUnit_ModifyBoneTransforms(
    BoneToModify=((Bone="upperarm_tricep_l",Transform=(Rotation=(X=0,Y=0,Z=0,W=1),Translation=(X=5,Y=0,Z=0),Scale3D=(X=1.3,Y=0.4,Z=0.4))),(Bone="None",Transform=(Rotation=(X=0,Y=0,Z=0,W=1),Translation=(X=0,Y=0,Z=0),Scale3D=(X=1,Y=1,Z=1))))
    Mode=AdditiveLocal
    WeightMaximum=1
    WeightMinimum=0
    Weight=&RigVMFunction_MathFloatRemap(SourceMaximum=120, SourceMinimum=20, TargetMinimum=0, bClamp=true, Value=@n3.Result.Z).Result
   )
   > /Script/ControlRig.RigUnit_TransformConstraint(
    BaseBone=clavicle_l
    BaseTransform=(Rotation=(X=0,Y=0,Z=0,W=1),Translation=(X=0,Y=0,Z=0),Scale3D=(X=1,Y=1,Z=1))
    BaseTransformSpace=GlobalSpace
    Bone=clavicle_float_l
    Targets="((Transform=(Rotation=(X=0,Y=0,Z=0,W=1),Translation=(X=0,Y=0,Z=0),Scale3D=(X=1,Y=1,Z=1)),Weight=0,bMaintainOffset=True,Filter=(TranslationFilter=(bX=True,bY=True,bZ=True),RotationFilter=(bX=True,bY=True,bZ=True),ScaleFilter=(bX=True,bY=True,bZ=True))),(Tra ...(truncated)"
    bUseInitialTransforms=True
   ) {is=struct RigVMUnitNode}
   > &RigUnit_TransformConstraint(
    BaseBone=None
    BaseTransform=(Rotation=(X=0,Y=0,Z=0,W=1),Translation=(X=0,Y=0,Z=0),Scale3D=(X=1,Y=1,Z=1))
    BaseTransformSpace=GlobalSpace
    Bone=clavicle_float_r
    Targets="((Transform=(Rotation=(X=0,Y=0,Z=0,W=1),Translation=(X=0,Y=0,Z=0),Scale3D=(X=1,Y=1,Z=1)),Weight=1,bMaintainOffset=True,Filter=(TranslationFilter=(bX=True,bY=True,bZ=True),RotationFilter=(bX=True,bY=True,bZ=True),ScaleFilter=(bX=True,bY=True,bZ=True))),(Tra ...(truncated)"
    bUseInitialTransforms=True
    Targets.1.Transform.Translation=&RigUnit_GetBoneTransform(Bone=clavicle_r).Transform.Translation
    Targets.0.Transform.Translation=&RigUnit_GetBoneTransform(Bone=upperarm_twist_02_r).Transform.Translation
   )
   n4 = &RigVMFunction_MathTransformMakeRelative(
    Global=(Rotation=(X=0,Y=0,Z=0,W=1),Translation=(X=0,Y=0,Z=0),Scale3D=(X=1,Y=1,Z=1))
    Parent=(Rotation=(X=0,Y=0,Z=0,W=1),Translation=(X=0,Y=0,Z=0),Scale3D=(X=1,Y=1,Z=1))
    Global.Translation=&RigUnit_GetBoneTransform(Bone=clavicle_float_r).Transform.Translation
    Parent.Translation=&RigUnit_GetBoneTransform(Bone=upperarm_r).Transform.Translation
   )
   n5 = /Script/RigVM.RigVMFunction_MathVectorLength(Value=@n4.Local.Translation) {is=struct RigVMUnitNode, template="Length::Execute(in Value,out Result)"}
   n6 = /Script/RigVM.RigVMFunction_MathFloatAbs(Value=&RigVMFunction_MathQuaternionToEuler(RotationOrder=ZXY, Value=&RigUnit_GetRelativeBoneTransform(Bone=upperarm_r, Space=clavicle_r).Transform.Rotation).Result.Z) {is=struct RigVMUnitNode, template="Absolute::Execute(in Value,out Result)"}
   > &RigUnit_ModifyBoneTransforms(
    BoneToModify=((Bone="clavicle_float_r",Transform=(Rotation=(X=0,Y=0,Z=0,W=1),Translation=(X=22,Y=-10,Z=-22),Scale3D=(X=1,Y=2.5,Z=2.5))),(Bone="None",Transform=(Rotation=(X=0,Y=0,Z=0,W=1),Translation=(X=0,Y=0,Z=0),Scale3D=(X=1,Y=1,Z=1))))
    Mode=AdditiveLocal
    WeightMaximum=1
    WeightMinimum=0
    Weight=&RigVMFunction_MathFloatMul(
     A=&RigVMFunction_MathFloatRemap(SourceMaximum=12.4, SourceMinimum=3.5, TargetMinimum=0, bClamp=true, Value=@n5.Result).Result
     B=&RigVMFunction_MathFloatRemap(SourceMaximum=125, SourceMinimum=65, TargetMinimum=0, bClamp=true, Value=@n6.Result).Result
    ) {template="Multiply::Execute(in A,in B,out Result)"}.Result
   )
   n7 = &RigVMFunction_MathTransformMakeRelative(
    Global=(Rotation=(X=0,Y=0,Z=0,W=1),Translation=(X=0,Y=0,Z=0),Scale3D=(X=1,Y=1,Z=1))
    Parent=(Rotation=(X=0,Y=0,Z=0,W=1),Translation=(X=0,Y=0,Z=0),Scale3D=(X=1,Y=1,Z=1))
    Global.Translation=&RigUnit_GetBoneTransform(Bone=clavicle_float_l).Transform.Translation
    Parent.Translation=&RigUnit_GetBoneTransform(Bone=upperarm_l).Transform.Translation
   )
   > &RigUnit_ModifyBoneTransforms(
    BoneToModify=((Bone="clavicle_float_l",Transform=(Rotation=(X=0,Y=0,Z=0,W=1),Translation=(X=-10,Y=10,Z=10),Scale3D=(X=1,Y=2,Z=2))),(Bone="None",Transform=(Rotation=(X=0,Y=0,Z=0,W=1),Translation=(X=0,Y=0,Z=0),Scale3D=(X=1,Y=1,Z=1))))
    Mode=AdditiveLocal
    WeightMaximum=1
    WeightMinimum=0
    Weight=&RigVMFunction_MathFloatRemap(
     SourceMaximum=11
     SourceMinimum=3.5
     TargetMinimum=0
     bClamp=true
     Value=&RigVMFunction_MathVectorLength(Value=@n7.Local.Translation) {template="Length::Execute(in Value,out Result)"}.Result
    ).Result
   )
   > &RigUnit_TransformConstraint(
    BaseBone=None
    BaseTransform=(Rotation=(X=0,Y=0,Z=0,W=1),Translation=(X=0,Y=0,Z=0),Scale3D=(X=1,Y=1,Z=1))
    BaseTransformSpace=GlobalSpace
    Bone=trapezius_l
    Targets="((Transform=(Rotation=(X=0,Y=0,Z=0,W=1),Translation=(X=0,Y=0,Z=0),Scale3D=(X=1,Y=1,Z=1)),Weight=1,bMaintainOffset=True,Filter=(TranslationFilter=(bX=True,bY=True,bZ=True),RotationFilter=(bX=True,bY=True,bZ=True),ScaleFilter=(bX=True,bY=True,bZ=True))),(Tra ...(truncated)"
    bUseInitialTransforms=True
    Targets.1.Transform=&RigUnit_GetBoneTransform(Bone=dyn_spineTrap).Transform
    Targets.0.Transform=&RigUnit_GetBoneTransform(Bone=clavicle_l).Transform
   )
   > &RigUnit_TransformConstraint(
    BaseBone=None
    BaseTransform=(Rotation=(X=0,Y=0,Z=0,W=1),Translation=(X=0,Y=0,Z=0),Scale3D=(X=1,Y=1,Z=1))
    BaseTransformSpace=GlobalSpace
    Bone=trapezius_r
    Targets="((Transform=(Rotation=(X=0,Y=0,Z=0,W=1),Translation=(X=0,Y=0,Z=0),Scale3D=(X=1,Y=1,Z=1)),Weight=0.5,bMaintainOffset=True,Filter=(TranslationFilter=(bX=True,bY=True,bZ=True),RotationFilter=(bX=True,bY=True,bZ=True),ScaleFilter=(bX=True,bY=True,bZ=True))),(T ...(truncated)"
    bUseInitialTransforms=True
    Targets.1.Transform=&RigUnit_GetBoneTransform(Bone=dyn_spineTrap).Transform
    Targets.0.Transform=&RigUnit_GetBoneTransform(Bone=clavicle_r).Transform
   )
   > &RigUnit_TransformConstraint(
    BaseBone=None
    BaseTransform=(Rotation=(X=0,Y=0,Z=0,W=1),Translation=(X=0,Y=0,Z=0),Scale3D=(X=1,Y=1,Z=1))
    BaseTransformSpace=GlobalSpace
    Bone=latissimus_l
    Targets="((Transform=(Rotation=(X=0,Y=0,Z=0,W=1),Translation=(X=0,Y=0,Z=0),Scale3D=(X=1,Y=1,Z=1)),Weight=1,bMaintainOffset=True,Filter=(TranslationFilter=(bX=True,bY=True,bZ=True),RotationFilter=(bX=True,bY=True,bZ=True),ScaleFilter=(bX=True,bY=True,bZ=True))),(Tra ...(truncated)"
    bUseInitialTransforms=True
    Targets.1.Transform=&RigUnit_GetBoneTransform(Bone=dyn_spineLat).Transform
    Targets.0.Transform=&RigUnit_GetBoneTransform(Bone=clavicle_l).Transform
   )
   > n8 = &RigUnit_TransformConstraint(
    BaseBone=None
    BaseTransform=(Rotation=(X=0,Y=0,Z=0,W=1),Translation=(X=0,Y=0,Z=0),Scale3D=(X=1,Y=1,Z=1))
    BaseTransformSpace=GlobalSpace
    Bone=latissimus_r
    Targets="((Transform=(Rotation=(X=0,Y=0,Z=0,W=1),Translation=(X=0,Y=0,Z=0),Scale3D=(X=1,Y=1,Z=1)),Weight=1,bMaintainOffset=True,Filter=(TranslationFilter=(bX=True,bY=True,bZ=True),RotationFilter=(bX=True,bY=True,bZ=True),ScaleFilter=(bX=True,bY=True,bZ=True))),(Tra ...(truncated)"
    bUseInitialTransforms=True
    Targets.1.Transform=&RigUnit_GetBoneTransform(Bone=dyn_spineLat).Transform
    Targets.0.Transform=&RigUnit_GetBoneTransform(Bone=clavicle_r).Transform
   )
   Branch_RigVMUnitNode = &RigVMFunction_ControlFlowBranch(ExecuteContext=@n8.ExecutePin, Condition=RigVMVariableNode() {variable=Crouch Walk}.Value)
   n9 = &RigVMFunction_MathFloatRemap(
    SourceMaximum=-32
    SourceMinimum=10
    TargetMinimum=0
    bClamp=true
    Value=&RigVMFunction_MathQuaternionToEuler(RotationOrder=ZXY, Value=&RigUnit_GetRelativeBoneTransform(Bone=clavicle_l, Space=spine_05).Transform.Rotation).Result.Z
   )
   &RigUnit_ModifyBoneTransforms(
    BoneToModify=((Bone="latissimus_l",Transform=(Rotation=(X=0,Y=0,Z=0,W=1),Translation=(X=0,Y=-0.75,Z=0),Scale3D=(X=0.85,Y=1.5,Z=1))),(Bone="None",Transform=(Rotation=(X=0,Y=0,Z=0,W=1),Translation=(X=0,Y=0,Z=0),Scale3D=(X=1,Y=1,Z=1))))
    Mode=AdditiveLocal
    WeightMaximum=1
    WeightMinimum=0
    ExecutePin=@Branch_RigVMUnitNode.False
    Weight=@n9.Result
   )
   > &RigUnit_ModifyBoneTransforms(
    BoneToModify=((Bone="trapezius_l",Transform=(Rotation=(X=0,Y=0,Z=0,W=1),Translation=(X=-2.75,Y=-2,Z=0),Scale3D=(X=0.825,Y=1.75,Z=1.1))),(Bone="None",Transform=(Rotation=(X=0,Y=0,Z=0,W=1),Translation=(X=0,Y=0,Z=0),Scale3D=(X=1,Y=1,Z=1))))
    Mode=AdditiveLocal
    WeightMaximum=1
    WeightMinimum=0
    Weight=@n9.Result
   )
   n10 = &RigUnit_GetRelativeBoneTransform(Bone=clavicle_r, Space=spine_05)
   n11 = &RigVMFunction_MathFloatRemap(
    SourceMaximum=170
    SourceMinimum=70
    TargetMinimum=0
    bClamp=true
    Value=&RigVMFunction_MathFloatAbs(Value=&RigVMFunction_MathQuaternionToEuler(RotationOrder=XYZ, Value=@n10.Transform.Rotation).Result.Z) {template="Absolute::Execute(in Value,out Result)"}.Result
   )
   > &RigUnit_ModifyBoneTransforms(
    BoneToModify=((Bone="latissimus_r",Transform=(Rotation=(X=0,Y=0,Z=0,W=1),Translation=(X=0,Y=0.75,Z=0),Scale3D=(X=0.85,Y=1.1,Z=1))),(Bone="None",Transform=(Rotation=(X=0,Y=0,Z=0,W=1),Translation=(X=0,Y=0,Z=0),Scale3D=(X=1,Y=1,Z=1))))
    Mode=AdditiveLocal
    WeightMaximum=1
    WeightMinimum=0
    Weight=@n11.Result
   )
   > &RigUnit_ModifyBoneTransforms(
    BoneToModify=((Bone="trapezius_r",Transform=(Rotation=(X=0,Y=0,Z=0,W=1),Translation=(X=2.75,Y=2,Z=0),Scale3D=(X=1,Y=1,Z=1))),(Bone="None",Transform=(Rotation=(X=0,Y=0,Z=0,W=1),Translation=(X=0,Y=0,Z=0),Scale3D=(X=1,Y=1,Z=1))))
    Mode=AdditiveLocal
    WeightMaximum=1
    WeightMinimum=0
    Weight=@n11.Result
   )
   n12 = &RigVMFunction_MathQuaternionToEuler(RotationOrder=XYZ, Value=&RigUnit_GetRelativeBoneTransform(Bone=lowerarm_r, Space=upperarm_r).Transform.Rotation)
   > &RigUnit_ModifyBoneTransforms(
    BoneToModify=((Bone="upperarm_bicep_r",Transform=(Rotation=(X=0,Y=0,Z=0,W=1),Translation=(X=6,Y=0.75,Z=0),Scale3D=(X=0.85,Y=2,Z=2))),(Bone="None",Transform=(Rotation=(X=0,Y=0,Z=0,W=1),Translation=(X=0,Y=0,Z=0),Scale3D=(X=1,Y=1,Z=1))))
    Mode=AdditiveLocal
    WeightMaximum=1
    WeightMinimum=0
    Weight=&RigVMFunction_MathFloatRemap(SourceMaximum=120, SourceMinimum=20, TargetMinimum=0, bClamp=true, Value=@n12.Result.Z).Result
   )
   > &RigUnit_ModifyBoneTransforms(
    BoneToModify=((Bone="upperarm_bicep_r",Transform=(Rotation=(X=0,Y=0,Z=0,W=1),Translation=(X=0,Y=1.5,Z=0),Scale3D=(X=1.1,Y=0.6,Z=0.6))),(Bone="None",Transform=(Rotation=(X=0,Y=0,Z=0,W=1),Translation=(X=0,Y=0,Z=0),Scale3D=(X=1,Y=1,Z=1))))
    Mode=AdditiveLocal
    WeightMaximum=1
    WeightMinimum=0
    Weight=&RigVMFunction_MathFloatRemap(SourceMaximum=20, SourceMinimum=120, TargetMinimum=0, bClamp=true, Value=@n12.Result.Z).Result
   )
   n13 = &RigVMFunction_MathQuaternionToEuler(RotationOrder=XYZ, Value=&RigUnit_GetRelativeBoneTransform(Bone=lowerarm_l, Space=upperarm_l).Transform.Rotation)
   > &RigUnit_ModifyBoneTransforms(
    BoneToModify=((Bone="upperarm_bicep_l",Transform=(Rotation=(X=0,Y=0,Z=0,W=1),Translation=(X=-6,Y=0.75,Z=0),Scale3D=(X=0.85,Y=1.6,Z=1.6))),(Bone="None",Transform=(Rotation=(X=0,Y=0,Z=0,W=1),Translation=(X=0,Y=0,Z=0),Scale3D=(X=1,Y=1,Z=1))))
    Mode=AdditiveLocal
    WeightMaximum=1
    WeightMinimum=0
    Weight=&RigVMFunction_MathFloatRemap(SourceMaximum=120, SourceMinimum=20, TargetMinimum=0, bClamp=true, Value=@n13.Result.Z).Result
   )
   > &RigUnit_ModifyBoneTransforms(
    BoneToModify=((Bone="upperarm_bicep_l",Transform=(Rotation=(X=0,Y=0,Z=0,W=1),Translation=(X=0,Y=-1.5,Z=0),Scale3D=(X=1.1,Y=0.6,Z=0.6))),(Bone="None",Transform=(Rotation=(X=0,Y=0,Z=0,W=1),Translation=(X=0,Y=0,Z=0),Scale3D=(X=1,Y=1,Z=1))))
    Mode=AdditiveLocal
    WeightMaximum=1
    WeightMinimum=0
    Weight=&RigVMFunction_MathFloatRemap(SourceMaximum=20, SourceMinimum=120, TargetMinimum=0, bClamp=true, Value=@n13.Result.Z).Result
   )
   n14 = /Script/RigVM.RigVMFunction_MathVectorAngle(
    B=/Script/RigVM.RigVMFunction_MathQuaternionGetAxis(Axis=Z, Quaternion=&RigUnit_GetBoneTransform(Bone=thigh_l).Transform.Rotation) {is=struct RigVMUnitNode, template="Axis::Execute(in Quaternion,in Axis,out Result)"}.Result
    A=&RigVMFunction_MathQuaternionGetAxis(
     Axis=Z
     Quaternion=/Script/ControlRig.RigUnit_GetInitialBoneTransform(Bone=thigh_l, Space=GlobalSpace) {is=struct RigVMUnitNode}.Transform.Rotation
    ) {template="Axis::Execute(in Quaternion,in Axis,out Result)"}.Result
   ) {is=struct RigVMUnitNode, template="AngleBetween::Execute(in A,in B,out Result)"}
   n15 = &RigVMFunction_MathQuaternionFromEuler(Euler=(X=0,Y=30,Z=0), RotationOrder=YXZ)
   > &RigUnit_ModifyBoneTransforms(
    BoneToModify=((Bone="dyn_belt_01_Base_l",Transform=(Rotation=(X=0,Y=0,Z=0,W=1),Translation=(X=0,Y=0,Z=0),Scale3D=(X=1,Y=1,Z=1))),(Bone="dyn_belt_01_Base_r",Transform=(Rotation=(X=0,Y=0,Z=0,W=1),Translation=(X=0,Y=0,Z=0),Scale3D=(X=1,Y=1,Z=1))))
    Mode=AdditiveLocal
    WeightMaximum=1
    WeightMinimum=0
    Weight=&RigVMFunction_MathFloatRemap(
     SourceMaximum=90
     SourceMinimum=0
     TargetMinimum=0
     bClamp=true
     Value=/Script/RigVM.RigVMFunction_MathFloatDeg(Value=@n14.Result) {is=struct RigVMUnitNode, template="Degrees::Execute(in Value,out Result)"}.Result
    ).Result
    BoneToModify.0.Transform.Rotation=@n15.Result
    BoneToModify.1.Transform.Rotation=@n15.Result
   )
   n16 = &RigVMFunction_MathVectorAngle(
    A=&RigVMFunction_MathQuaternionGetAxis(Axis=Z, Quaternion=&RigUnit_GetInitialBoneTransform(Bone=thigh_r, Space=GlobalSpace).Transform.Rotation) {template="Axis::Execute(in Quaternion,in Axis,out Result)"}.Result
    B=&RigVMFunction_MathQuaternionGetAxis(Axis=Z, Quaternion=&RigUnit_GetBoneTransform(Bone=thigh_r).Transform.Rotation) {template="Axis::Execute(in Quaternion,in Axis,out Result)"}.Result
   ) {template="AngleBetween::Execute(in A,in B,out Result)"}
   n17 = &RigVMFunction_MathQuaternionFromEuler(Euler=(X=0,Y=-30,Z=0), RotationOrder=YXZ)
   > &RigUnit_ModifyBoneTransforms(
    BoneToModify=((Bone="dyn_belt_01_Base_l",Transform=(Rotation=(X=0,Y=0,Z=0,W=1),Translation=(X=0,Y=0,Z=0),Scale3D=(X=1,Y=1,Z=1))),(Bone="dyn_belt_01_Base_r",Transform=(Rotation=(X=0,Y=0,Z=0,W=1),Translation=(X=0,Y=0,Z=0),Scale3D=(X=1,Y=1,Z=1))))
    Mode=AdditiveLocal
    WeightMaximum=1
    WeightMinimum=0
    Weight=&RigVMFunction_MathFloatRemap(
     SourceMaximum=90
     SourceMinimum=0
     TargetMinimum=0
     bClamp=true
     Value=&RigVMFunction_MathFloatDeg(Value=@n16.Result) {template="Degrees::Execute(in Value,out Result)"}.Result
    ).Result
    BoneToModify.0.Transform.Rotation=@n17.Result
    BoneToModify.1.Transform.Rotation=@n17.Result
   )
   n18 = &RigUnit_GetBoneTransform(Bone=dyn_thighsLookAt)
   n19 = &RigVMFunction_MathQuaternionGetAxis(Axis=X, Quaternion=@n18.Transform.Rotation) {template="Axis::Execute(in Quaternion,in Axis,out Result)"}
   n20 = /Script/RigVM.RigVMFunction_MathVectorSub(
    B=@n18.Transform.Translation
    A=/Script/RigVM.RigVMFunction_MathTransformTransformVector(Location=(X=0,Y=0,Z=0), Transform=&RigUnit_GetBoneTransform(Bone=thigh_twist_01_l).Transform) {is=struct RigVMUnitNode}.Result
   ) {is=struct RigVMUnitNode, template="Subtract::Execute(in A,in B,out Result)"}
   > /Script/ControlRig.RigUnit_SetBoneRotation(
    Bone=dyn_thighsLookAt
    Space=GlobalSpace
    Weight=1
    bPropagateToChildren=True
    Rotation=/Script/RigVM.RigVMFunction_MathQuaternionMul(
     B=@n18.Transform.Rotation
     A=/Script/RigVM.RigVMFunction_MathQuaternionFromTwoVectors(A=@n19.Result, B=@n20.Result) {is=struct RigVMUnitNode, template="FromTwoVectors::Execute(in A,in B,out Result)"}.Result
    ) {is=struct RigVMUnitNode, template="Multiply::Execute(in A,in B,out Result)"}.Result
   ) {is=struct RigVMUnitNode}
   n21 = &RigVMFunction_MathFloatRemap(
    SourceMaximum=-32
    SourceMinimum=10
    TargetMinimum=0
    bClamp=true
    Value=&RigVMFunction_MathQuaternionToEuler(RotationOrder=ZXY, Value=&RigUnit_GetRelativeBoneTransform(Bone=clavicle_l, Space=spine_05).Transform.Rotation).Result.Z
   )
   &RigUnit_ModifyBoneTransforms(
    BoneToModify=((Bone="latissimus_l",Transform=(Rotation=(X=0,Y=0,Z=0,W=1),Translation=(X=0,Y=-0.75,Z=0),Scale3D=(X=0.85,Y=1.5,Z=1))),(Bone="None",Transform=(Rotation=(X=0,Y=0,Z=0,W=1),Translation=(X=0,Y=0,Z=0),Scale3D=(X=1,Y=1,Z=1))))
    Mode=AdditiveLocal
    WeightMaximum=1
    WeightMinimum=0
    ExecutePin=@Branch_RigVMUnitNode.True
    Weight=@n21.Result
   )
   > &RigUnit_ModifyBoneTransforms(
    BoneToModify=((Bone="trapezius_l",Transform=(Rotation=(X=0,Y=0,Z=0,W=1),Translation=(X=-2.75,Y=-2,Z=0),Scale3D=(X=0.825,Y=1.75,Z=1.1))),(Bone="None",Transform=(Rotation=(X=0,Y=0,Z=0,W=1),Translation=(X=0,Y=0,Z=0),Scale3D=(X=1,Y=1,Z=1))))
    Mode=AdditiveLocal
    WeightMaximum=1
    WeightMinimum=0
    Weight=@n21.Result
   )
   n22 = &RigUnit_GetRelativeBoneTransform(Bone=clavicle_r, Space=spine_05)
   n23 = &RigVMFunction_MathFloatRemap(
    SourceMaximum=170
    SourceMinimum=70
    TargetMinimum=0
    bClamp=true
    Value=&RigVMFunction_MathFloatAbs(Value=&RigVMFunction_MathQuaternionToEuler(RotationOrder=XYZ, Value=@n22.Transform.Rotation).Result.Z) {template="Absolute::Execute(in Value,out Result)"}.Result
   )
   > &RigUnit_ModifyBoneTransforms(
    BoneToModify=((Bone="latissimus_r",Transform=(Rotation=(X=0,Y=0,Z=0,W=1),Translation=(X=0,Y=0.75,Z=0),Scale3D=(X=1,Y=1.1,Z=1))),(Bone="None",Transform=(Rotation=(X=0,Y=0,Z=0,W=1),Translation=(X=0,Y=0,Z=0),Scale3D=(X=1,Y=1,Z=1))))
    Mode=AdditiveLocal
    WeightMaximum=1
    WeightMinimum=0
    Weight=@n23.Result
   )
   > &RigUnit_ModifyBoneTransforms(
    BoneToModify=((Bone="trapezius_r",Transform=(Rotation=(X=0,Y=0,Z=0,W=1),Translation=(X=0.75,Y=0,Z=0),Scale3D=(X=1,Y=1.3,Z=1.1))),(Bone="None",Transform=(Rotation=(X=0,Y=0,Z=0,W=1),Translation=(X=0,Y=0,Z=0),Scale3D=(X=1,Y=1,Z=1))))
    Mode=AdditiveLocal
    WeightMaximum=1
    WeightMinimum=0
    Weight=@n23.Result
   )
   n24 = &RigVMFunction_MathQuaternionToEuler(RotationOrder=XYZ, Value=&RigUnit_GetRelativeBoneTransform(Bone=lowerarm_r, Space=upperarm_r).Transform.Rotation)
   > &RigUnit_ModifyBoneTransforms(
    BoneToModify=((Bone="upperarm_bicep_r",Transform=(Rotation=(X=0,Y=0,Z=0,W=1),Translation=(X=6,Y=0.75,Z=0),Scale3D=(X=0.85,Y=2,Z=2))),(Bone="None",Transform=(Rotation=(X=0,Y=0,Z=0,W=1),Translation=(X=0,Y=0,Z=0),Scale3D=(X=1,Y=1,Z=1))))
    Mode=AdditiveLocal
    WeightMaximum=1
    WeightMinimum=0
    Weight=&RigVMFunction_MathFloatRemap(SourceMaximum=120, SourceMinimum=20, TargetMinimum=0, bClamp=true, Value=@n24.Result.Z).Result
   )
   > &RigUnit_ModifyBoneTransforms(
    BoneToModify=((Bone="upperarm_bicep_r",Transform=(Rotation=(X=0,Y=0,Z=0,W=1),Translation=(X=0,Y=1.5,Z=0),Scale3D=(X=1.1,Y=0.6,Z=0.6))),(Bone="None",Transform=(Rotation=(X=0,Y=0,Z=0,W=1),Translation=(X=0,Y=0,Z=0),Scale3D=(X=1,Y=1,Z=1))))
    Mode=AdditiveLocal
    WeightMaximum=1
    WeightMinimum=0
    Weight=&RigVMFunction_MathFloatRemap(SourceMaximum=20, SourceMinimum=120, TargetMinimum=0, bClamp=true, Value=@n24.Result.Z).Result
   )
   n25 = &RigVMFunction_MathQuaternionToEuler(RotationOrder=XYZ, Value=&RigUnit_GetRelativeBoneTransform(Bone=lowerarm_l, Space=upperarm_l).Transform.Rotation)
   > &RigUnit_ModifyBoneTransforms(
    BoneToModify=((Bone="upperarm_bicep_l",Transform=(Rotation=(X=0,Y=0,Z=0,W=1),Translation=(X=-6,Y=0.75,Z=0),Scale3D=(X=0.85,Y=1.6,Z=1.6))),(Bone="None",Transform=(Rotation=(X=0,Y=0,Z=0,W=1),Translation=(X=0,Y=0,Z=0),Scale3D=(X=1,Y=1,Z=1))))
    Mode=AdditiveLocal
    WeightMaximum=1
    WeightMinimum=0
    Weight=&RigVMFunction_MathFloatRemap(SourceMaximum=120, SourceMinimum=20, TargetMinimum=0, bClamp=true, Value=@n25.Result.Z).Result
   )
   > &RigUnit_ModifyBoneTransforms(
    BoneToModify=((Bone="upperarm_bicep_l",Transform=(Rotation=(X=0,Y=0,Z=0,W=1),Translation=(X=0,Y=-1.5,Z=0),Scale3D=(X=1.1,Y=0.6,Z=0.6))),(Bone="None",Transform=(Rotation=(X=0,Y=0,Z=0,W=1),Translation=(X=0,Y=0,Z=0),Scale3D=(X=1,Y=1,Z=1))))
    Mode=AdditiveLocal
    WeightMaximum=1
    WeightMinimum=0
    Weight=&RigVMFunction_MathFloatRemap(SourceMaximum=20, SourceMinimum=120, TargetMinimum=0, bClamp=true, Value=@n25.Result.Z).Result
   )
   n26 = &RigVMFunction_MathVectorAngle(
    A=&RigVMFunction_MathQuaternionGetAxis(Axis=Z, Quaternion=&RigUnit_GetInitialBoneTransform(Bone=thigh_l, Space=GlobalSpace).Transform.Rotation) {template="Axis::Execute(in Quaternion,in Axis,out Result)"}.Result
    B=&RigVMFunction_MathQuaternionGetAxis(Axis=Z, Quaternion=&RigUnit_GetBoneTransform(Bone=thigh_l).Transform.Rotation) {template="Axis::Execute(in Quaternion,in Axis,out Result)"}.Result
   ) {template="AngleBetween::Execute(in A,in B,out Result)"}
   n27 = &RigVMFunction_MathQuaternionFromEuler(Euler=(X=0,Y=30,Z=0), RotationOrder=YXZ)
   > &RigUnit_ModifyBoneTransforms(
    BoneToModify=((Bone="dyn_belt_01_Base_l",Transform=(Rotation=(X=0,Y=0,Z=0,W=1),Translation=(X=0,Y=0,Z=0),Scale3D=(X=1,Y=1,Z=1))),(Bone="dyn_belt_01_Base_r",Transform=(Rotation=(X=0,Y=0,Z=0,W=1),Translation=(X=0,Y=0,Z=0),Scale3D=(X=1,Y=1,Z=1))))
    Mode=AdditiveLocal
    WeightMaximum=1
    WeightMinimum=0
    Weight=&RigVMFunction_MathFloatRemap(
     SourceMaximum=90
     SourceMinimum=0
     TargetMinimum=0
     bClamp=true
     Value=&RigVMFunction_MathFloatDeg(Value=@n26.Result) {template="Degrees::Execute(in Value,out Result)"}.Result
    ).Result
    BoneToModify.0.Transform.Rotation=@n27.Result
    BoneToModify.1.Transform.Rotation=@n27.Result
   )
   n28 = &RigVMFunction_MathVectorAngle(
    A=&RigVMFunction_MathQuaternionGetAxis(Axis=Z, Quaternion=&RigUnit_GetInitialBoneTransform(Bone=thigh_r, Space=GlobalSpace).Transform.Rotation) {template="Axis::Execute(in Quaternion,in Axis,out Result)"}.Result
    B=&RigVMFunction_MathQuaternionGetAxis(Axis=Z, Quaternion=&RigUnit_GetBoneTransform(Bone=thigh_r).Transform.Rotation) {template="Axis::Execute(in Quaternion,in Axis,out Result)"}.Result
   ) {template="AngleBetween::Execute(in A,in B,out Result)"}
   n29 = &RigVMFunction_MathQuaternionFromEuler(Euler=(X=0,Y=-30,Z=0), RotationOrder=YXZ)
   > &RigUnit_ModifyBoneTransforms(
    BoneToModify=((Bone="dyn_belt_01_Base_l",Transform=(Rotation=(X=0,Y=0,Z=0,W=1),Translation=(X=0,Y=0,Z=0),Scale3D=(X=1,Y=1,Z=1))),(Bone="dyn_belt_01_Base_r",Transform=(Rotation=(X=0,Y=0,Z=0,W=1),Translation=(X=0,Y=0,Z=0),Scale3D=(X=1,Y=1,Z=1))))
    Mode=AdditiveLocal
    WeightMaximum=1
    WeightMinimum=0
    Weight=&RigVMFunction_MathFloatRemap(
     SourceMaximum=90
     SourceMinimum=0
     TargetMinimum=0
     bClamp=true
     Value=&RigVMFunction_MathFloatDeg(Value=@n28.Result) {template="Degrees::Execute(in Value,out Result)"}.Result
    ).Result
    BoneToModify.0.Transform.Rotation=@n29.Result
    BoneToModify.1.Transform.Rotation=@n29.Result
   )
   n30 = &RigUnit_GetBoneTransform(Bone=dyn_thighsLookAt)
   n31 = &RigVMFunction_MathQuaternionGetAxis(Axis=X, Quaternion=@n30.Transform.Rotation) {template="Axis::Execute(in Quaternion,in Axis,out Result)"}
   n32 = &RigVMFunction_MathVectorSub(
    B=@n30.Transform.Translation
    A=&RigVMFunction_MathTransformTransformVector(Location=(X=0,Y=0,Z=0), Transform=&RigUnit_GetBoneTransform(Bone=thigh_twist_01_l).Transform).Result
   ) {template="Subtract::Execute(in A,in B,out Result)"}
   > &RigUnit_SetBoneRotation(
    Bone=dyn_thighsLookAt
    Space=GlobalSpace
    Weight=1
    bPropagateToChildren=True
    Rotation=&RigVMFunction_MathQuaternionMul(
     B=@n30.Transform.Rotation
     A=&RigVMFunction_MathQuaternionFromTwoVectors(A=@n31.Result, B=@n32.Result) {template="FromTwoVectors::Execute(in A,in B,out Result)"}.Result
    ) {template="Multiply::Execute(in A,in B,out Result)"}.Result
   )
   /Script/ControlRig.RigUnit_DebugTransformMutable(
    Color=(R=1,G=0,B=0,A=1)
    DebugDrawSettings=(DepthPriority=SDPG_Foreground,Lifetime=-1)
    ExecutePin=()
    Mode=Box
    Scale=20
    Space=None
    Thickness=0
    WorldOffset=(Rotation=(X=0,Y=0,Z=0,W=1),Translation=(X=0,Y=0,Z=0),Scale3D=(X=1,Y=1,Z=1))
    bEnabled=True
    Transform=&RigUnit_GetBoneTransform(Bone=dyn_thighsLookAt).Transform
   ) {is=struct RigVMUnitNode}
   &RigUnit_DebugTransformMutable(
    Color=(R=1,G=0,B=0,A=1)
    DebugDrawSettings=(DepthPriority=SDPG_Foreground,Lifetime=-1)
    ExecutePin=()
    Mode=Box
    Scale=20
    Space=None
    Thickness=0
    WorldOffset=(Rotation=(X=0,Y=0,Z=0,W=1),Translation=(X=0,Y=0,Z=0),Scale3D=(X=1,Y=1,Z=1))
    bEnabled=True
    Transform=&RigUnit_GetBoneTransform(Bone=dyn_thighsLookAt).Transform
   )
   &RigUnit_ModifyBoneTransforms(
    BoneToModify=((Bone="trapezius_r",Transform=(Rotation=(X=0,Y=0,Z=0,W=1),Translation=(X=2.75,Y=2,Z=0),Scale3D=(X=0.825,Y=1.75,Z=1.1))),(Bone="None",Transform=(Rotation=(X=0,Y=0,Z=0,W=1),Translation=(X=0,Y=0,Z=0),Scale3D=(X=1,Y=1,Z=1))))
    ExecutePin=()
    Mode=AdditiveLocal
    Weight=0
    WeightMaximum=1
    WeightMinimum=0
   )
  ]}
]
hierarchy [
 root {loc=<0 0 0>, rot=<0 0 0 1>}
 attach {parent=root, loc=<0 0 0>, rot=<0 0 0 1>}
 pelvis {parent=root, loc=<0 -0.7071 93.1191>, rot=<0.0127 0.707 -0.0127 -0.707>}
 spine_01 {parent=pelvis, loc=<6.7494 -0.6388 0>, rot=<0 0 0.034 -0.9994>}
 spine_02 {parent=spine_01, loc=<7.2518 -0.035 0>, rot=<0 0 0.0238 -0.9997>}
 spine_03 {parent=spine_02, loc=<6.0104 0 0>, rot=<0 0 0.0149 -0.9999>}
 spine_04 {parent=spine_03, loc=<8.1024 0 0>, rot=<0 0 -0.0132 -0.9999>}
 spine_05 {parent=spine_04, loc=<10.0146 0 0>, rot=<0 0 -0.0015 -1>}
 clavicle_l {parent=spine_05, loc=<4.3575 -0.1466 -2.3223>, rot=<0.0465 -0.7289 -0.009 -0.6829>}
 upperarm_l {parent=clavicle_l, loc=<15.5111 0 0>, rot=<0.0226 -0.3234 -0.0088 -0.9459>}
 lowerarm_l {parent=upperarm_l, loc=<26.6514 0 0>, rot=<0.0391 -0.0099 -0.2453 -0.9686>}
 hand_l {parent=lowerarm_l, loc=<22.3491 0 0>, rot=<-0.7081 -0.0507 0.0975 -0.6975>}
 index_metacarpal_l {parent=hand_l, loc=<1.6535 0.0183 -2.96>, rot=<0.0158 -0.0124 -0.0173 -0.9996>}
 index_01_l {parent=index_metacarpal_l, loc=<8.3008 0.0357 -0.0524>, rot=<-0.0044 0.0485 0.0924 -0.9945>}
 index_02_l {parent=index_01_l, loc=<3.3587 -0.3004 0.0452>, rot=<-0.002 0.0027 0.3225 -0.9466>}
 index_03_l {parent=index_02_l, loc=<2.5578 -0.123 -0.0252>, rot=<-0.002 0.0067 0.2558 -0.9667>}
 middle_metacarpal_l {parent=hand_l, loc=<2.1856 -0.1725 -0.5984>, rot=<0.006 0.0168 -0.0329 -0.9993>}
 middle_01_l {parent=middle_metacarpal_l, loc=<7.3275 -0.2447 0.0129>, rot=<-0.0208 0.0042 0.0858 -0.9961>}
 middle_02_l {parent=middle_01_l, loc=<4.0123 0.1401 0.1175>, rot=<-0.0218 0.0258 0.3512 -0.9357>}
 middle_03_l {parent=middle_02_l, loc=<2.6077 0.0534 0.1373>, rot=<0.0082 0.0188 0.1708 -0.9851>}
 pinky_metacarpal_l {parent=hand_l, loc=<1.7616 -0.9753 3.3814>, rot=<-0.0397 0.0814 -0.069 -0.9935>}
 pinky_01_l {parent=pinky_metacarpal_l, loc=<7.1709 -0.4831 -0.1094>, rot=<-0.0654 -0.0035 0.2242 -0.9723>}
 pinky_02_l {parent=pinky_01_l, loc=<2.8274 0.0893 -0.1043>, rot=<0.0037 0.0011 0.2854 -0.9584>}
 pinky_03_l {parent=pinky_02_l, loc=<2.2046 0.0307 -0.0078>, rot=<-0.0002 0.0098 0.1445 -0.9895>}
 ring_metacarpal_l {parent=hand_l, loc=<2.2074 -0.3625 1.3638>, rot=<-0.0284 0.0722 -0.0458 -0.9959>}
 ring_01_l {parent=ring_metacarpal_l, loc=<7.0295 -0.6197 -0.1446>, rot=<-0.0489 -0.0252 0.1892 -0.9804>}
 ring_02_l {parent=ring_01_l, loc=<3.6174 0.4646 -0.1027>, rot=<0 0 0.2806 -0.9598>}
 ring_03_l {parent=ring_02_l, loc=<2.4152 0.1141 0.0199>, rot=<0 0 0.1189 -0.9929>}
 thumb_01_l {parent=hand_l, loc=<1.3118 -0.3191 -2.459>, rot=<0.6546 -0.4042 0.0664 -0.6354>}
 thumb_02_l {parent=thumb_01_l, loc=<5.9336 -0.0448 0.0401>, rot=<-0.0037 0.0213 0.298 -0.9543>}
 thumb_03_l {parent=thumb_02_l, loc=<2.2223 0.3682 -0.0907>, rot=<0.0118 -0.013 0.0066 -0.9998>}
 weapon_l {parent=hand_l, loc=<-0.1974 -3.4111 0.3811>, rot=<0 0 0 1>}
 lowerarm_twist_01_l {parent=lowerarm_l, loc=<17.8549 0 0>, rot=<0 0 0 1>}
 lowerarm_twist_02_l {parent=lowerarm_l, loc=<9.6199 0 0>, rot=<0 0 0 1>}
 upperarm_twist_01_l {parent=upperarm_l, loc=<0.0041 0 0>, rot=<0 0 0 -1>}
 upperarm_twist_02_l {parent=upperarm_l, loc=<13.7941 0 0>, rot=<0 0 0 -1>}
 upperarm_tricepBase_l {parent=upperarm_twist_02_l, loc=<-0.0003 -5.3145 0.0004>, rot=<0 0 0 -1>}
 upperarm_tricep_l {parent=upperarm_tricepBase_l, loc=<0 0 0>, rot=<0 0 0 -1>}
 upperarm_bicepBase_l {parent=upperarm_twist_02_l, loc=<-0.0003 4.15 0.0004>, rot=<0 0 0 -1>}
 upperarm_bicep_l {parent=upperarm_bicepBase_l, loc=<0 -3 0>, rot=<0 0 0 1>, scale=<0.85 0.85 0.85>}
 clavicle_float_l {parent=clavicle_l, loc=<15 0 0>, rot=<0 0 0 1>}
 clavicle_r {parent=spine_05, loc=<4.3575 -0.1466 2.3223>, rot=<0.7289 0.0465 -0.6829 0.009>}
 upperarm_r {parent=clavicle_r, loc=<-15.5111 0 0>, rot=<0.0226 -0.3234 -0.0088 -0.9459>}
 lowerarm_r {parent=upperarm_r, loc=<-26.6514 0 0>, rot=<0.0391 -0.0099 -0.2453 -0.9686>}
 hand_r {parent=lowerarm_r, loc=<-22.3492 0 0>, rot=<-0.7081 -0.0507 0.0975 -0.6975>}
 index_metacarpal_r {parent=hand_r, loc=<-1.6534 -0.0181 2.9601>, rot=<0.0158 -0.0124 -0.0173 -0.9996>}
 index_01_r {parent=index_metacarpal_r, loc=<-8.3008 -0.0357 0.0524>, rot=<-0.0044 0.0485 0.0924 -0.9945>}
 index_02_r {parent=index_01_r, loc=<-3.3589 0.3003 -0.0453>, rot=<-0.002 0.0027 0.3225 -0.9466>}
 index_03_r {parent=index_02_r, loc=<-2.5578 0.1231 0.0251>, rot=<-0.002 0.0067 0.2558 -0.9667>}
 middle_metacarpal_r {parent=hand_r, loc=<-2.1856 0.1726 0.5984>, rot=<0.006 0.0168 -0.0329 -0.9993>}
 middle_01_r {parent=middle_metacarpal_r, loc=<-7.3273 0.2447 -0.0129>, rot=<-0.0208 0.0042 0.0858 -0.9961>}
 middle_02_r {parent=middle_01_r, loc=<-4.0142 -0.1411 -0.07>, rot=<-0.0217 0.0259 0.3512 -0.9357>}
 middle_03_r {parent=middle_02_r, loc=<-2.6074 -0.0534 -0.1312>, rot=<0.0083 0.0188 0.1708 -0.9851>}
 pinky_metacarpal_r {parent=hand_r, loc=<-1.7617 0.9753 -3.3814>, rot=<-0.0397 0.0814 -0.069 -0.9935>}
 pinky_01_r {parent=pinky_metacarpal_r, loc=<-7.1707 0.4832 0.1094>, rot=<-0.0654 -0.0035 0.2242 -0.9723>}
 pinky_02_r {parent=pinky_01_r, loc=<-2.8274 -0.0892 0.1043>, rot=<0.0037 0.0011 0.2854 -0.9584>}
 pinky_03_r {parent=pinky_02_r, loc=<-2.2045 -0.031 0.0078>, rot=<-0.0002 0.0098 0.1445 -0.9895>}
 ring_metacarpal_r {parent=hand_r, loc=<-2.2073 0.3627 -1.3638>, rot=<-0.0284 0.0722 -0.0458 -0.9959>}
 ring_01_r {parent=ring_metacarpal_r, loc=<-7.0297 0.6197 0.1446>, rot=<-0.0489 -0.0252 0.1892 -0.9804>}
 ring_02_r {parent=ring_01_r, loc=<-3.6172 -0.4648 0.1027>, rot=<0 0 0.2806 -0.9598>}
 ring_03_r {parent=ring_02_r, loc=<-2.4152 -0.1141 -0.0199>, rot=<0 0 0.1189 -0.9929>}
 thumb_01_r {parent=hand_r, loc=<-1.3119 0.3192 2.459>, rot=<0.6546 -0.4042 0.0664 -0.6354>}
 thumb_02_r {parent=thumb_01_r, loc=<-5.9335 0.0447 -0.0401>, rot=<-0.0037 0.0213 0.298 -0.9543>}
 thumb_03_r {parent=thumb_02_r, loc=<-2.2223 -0.3682 0.0907>, rot=<0.0118 -0.013 0.0066 -0.9998>}
 weapon_r {parent=hand_r, loc=<0.1975 3.4111 -0.3811>, rot=<0 0 0 1>}
 lowerarm_twist_01_r {parent=lowerarm_r, loc=<-17.855 0 0>, rot=<0 0 0 1>}
 lowerarm_twist_02_r {parent=lowerarm_r, loc=<-9.6199 0 0>, rot=<0 0 0 1>}
 upperarm_twist_01_r {parent=upperarm_r, loc=<-0.0041 0 0>, rot=<0 0 0 1>}
 upperarm_twist_02_r {parent=upperarm_r, loc=<-13.7942 0 0>, rot=<0 0 0 1>}
 upperarm_tricepBase_r {parent=upperarm_twist_02_r, loc=<0 5.3145 0>, rot=<0 0 0 1>}
 upperarm_tricep_r {parent=upperarm_tricepBase_r, loc=<0 0 0>, rot=<0 0 0 1>}
 upperarm_bicepBase_r {parent=upperarm_twist_02_r, loc=<0 -4.15 0>, rot=<0 0 0 1>}
 upperarm_bicep_r {parent=upperarm_bicepBase_r, loc=<0 0 0>, rot=<0 0 0 1>}
 clavicle_float_r {parent=clavicle_r, loc=<-15 0 0>, rot=<0 0 0 -1>}
 neck_01 {parent=spine_05, loc=<11.3138 0.1529 0>, rot=<0 0 -0.2003 -0.9797>}
 neck_02 {parent=neck_01, loc=<4.5627 0 0>, rot=<0 0 0.0504 -0.9987>}
 head {parent=neck_02, loc=<5.3914 0 0>, rot=<0 0 0.115 -0.9934>}
 dyn_spineTrap {parent=spine_05, loc=<-0.8058 -11.9587 0.001>, rot=<0 0 0 -1>}
 trapezius_l {parent=dyn_spineTrap, loc=<3.4 0 -5.362>, rot=<0.0465 -0.7289 -0.009 -0.6829>}
 trapezius_r {parent=dyn_spineTrap, loc=<3.3996 0 5.3615>, rot=<0.7289 0.0465 -0.6829 0.009>}
 dyn_spineLat {parent=spine_04, loc=<4.3871 -11.5586 0.001>, rot=<0 0 -0.0015 -1>}
 latissimus_l {parent=dyn_spineLat, loc=<-0.8746 1 -9.8593>, rot=<0.0465 -0.7289 -0.009 -0.6829>}
 latissimus_r {parent=dyn_spineLat, loc=<-0.8746 1 9.8593>, rot=<0.7289 0.0465 -0.6829 0.009>}
 thigh_l {parent=pelvis, loc=<-2.5693 -0.0109 -7.7501>, rot=<0.0012 0.0572 0.0155 -0.9982>}
 calf_l {parent=thigh_l, loc=<-39.0484 0 0>, rot=<0.0006 -0.0161 -0.0391 -0.9991>}
 calf_twist_01_l {parent=calf_l, loc=<-35.8131 -0.0967 0.0397>, rot=<0.0237 0.0095 -0.0035 -0.9997>}
 calf_twist_02_l {parent=calf_l, loc=<-18.8064 -0.0967 0.0397>, rot=<0.0237 0.0095 -0.0035 -0.9997>}
 foot_l {parent=calf_l, loc=<-42.561 0.1077 0.4292>, rot=<0.0418 -0.0455 0.0343 -0.9975>}
 ball_l {parent=foot_l, loc=<-6.1599 13.8264 0.6806>, rot=<0.0001 -0.0001 -0.7063 -0.7079>}
 calf_scale_l {parent=calf_l, loc=<0 0 0>, rot=<0 0 0 1>}
 thigh_twist_01_l {parent=thigh_l, loc=<-21.6303 0 0>, rot=<0.0013 0 0 -1>}
 thigh_scale_l {parent=thigh_l, loc=<0 0 0>, rot=<0 0 0 1>}
 thigh_r {parent=pelvis, loc=<-2.5691 -0.0109 7.7501>, rot=<0.0572 -0.0012 0.9982 0.0155>}
 calf_r {parent=thigh_r, loc=<39.0487 0 0>, rot=<0.0006 -0.0161 -0.0391 -0.9991>}
 calf_twist_01_r {parent=calf_r, loc=<35.8132 0.0967 -0.0397>, rot=<0.0237 0.0095 -0.0035 -0.9997>}
 calf_twist_02_r {parent=calf_r, loc=<18.8064 0.0967 -0.0397>, rot=<0.0237 0.0095 -0.0035 -0.9997>}
 foot_r {parent=calf_r, loc=<42.5611 -0.1077 -0.4292>, rot=<0.0418 -0.0455 0.0343 -0.9975>}
 ball_r {parent=foot_r, loc=<6.1598 -13.8264 -0.6806>, rot=<0.0001 -0.0001 -0.7063 -0.7079>}
 calf_scale_r {parent=calf_r, loc=<0 0 0>, rot=<0 0 0 1>}
 thigh_twist_01_r {parent=thigh_r, loc=<21.6303 0 0>, rot=<0.0013 0 0 -1>}
 dyn_thighsLookAt {parent=thigh_twist_01_r, loc=<0 0 0>, rot=<-0.4806 -0.518 0.5373 -0.4604>}
 thigh_scale_r {parent=thigh_r, loc=<0 0 0>, rot=<0 0 0 1>}
 dyn_belt_01_Base_r {parent=pelvis, loc=<9.2524 11.9646 -0.2148>, rot=<-0.1431 -0.9871 0.0078 0.0715>}
 dyn_belt_01_r {parent=dyn_belt_01_Base_r, loc=<0 0 0>, rot=<0 0 0 1>}
 dyn_belt_02_r {parent=dyn_belt_01_r, loc=<5.4901 0 0>, rot=<-0.002 -0.0074 -0.0086 -0.9999>}
 dyn_belt_03_r {parent=dyn_belt_02_r, loc=<6.1308 0 0>, rot=<0.0023 0.0088 0.0085 -0.9999>}
 dyn_belt_04_r {parent=dyn_belt_03_r, loc=<6.7253 0 0>, rot=<-0.0027 -0.01 -0.0077 -0.9999>}
 dyn_belt_05_r {parent=dyn_belt_04_r, loc=<6.8549 0 0>, rot=<0.0033 0.0123 0.0086 -0.9999>}
 dyn_belt_06_r {parent=dyn_belt_05_r, loc=<6.9188 0 0>, rot=<0.02 0.0737 -0.015 -0.997>}
 dyn_belt_01_Base_l {parent=pelvis, loc=<7.7341 13.6813 -1.0549>, rot=<-0.1618 -0.9825 -0.0117 -0.0914>}
 dyn_belt_01_l {parent=dyn_belt_01_Base_l, loc=<0 0 0>, rot=<0 0 0 -1>}
 dyn_belt_02_l {parent=dyn_belt_01_l, loc=<5.3848 0 0>, rot=<-0.0013 -0.0047 0.0244 -0.9997>}
 dyn_belt_03_l {parent=dyn_belt_02_l, loc=<6.2155 0 0>, rot=<0.0009 0.0035 -0.0022 -1>}
 dyn_belt_04_l {parent=dyn_belt_03_l, loc=<6.8109 0 0>, rot=<-0.0023 -0.0089 -0.004 -1>}
 dyn_belt_05_l {parent=dyn_belt_04_l, loc=<6.6539 0 0>, rot=<0.0011 0.0043 0 -1>}
 dyn_belt_06_l {parent=dyn_belt_05_l, loc=<6.918 0 0>, rot=<-0.0025 -0.0102 0.0118 -0.9999>}
 ik_foot_root {parent=root, loc=<0 0 0>, rot=<0 0 0 1>}
 ik_foot_l {parent=ik_foot_root, loc=<15.2608 -4.4176 9.4433>, rot=<-0.0357 -0.7045 -0.0284 0.7082>}
 ik_foot_r {parent=ik_foot_root, loc=<-15.2609 -4.4176 9.4431>, rot=<-0.7082 -0.0284 0.7045 -0.0357>}
 ik_hand_root {parent=root, loc=<0 0 0>, rot=<0 0 0 1>}
 ik_hand_gun {parent=ik_hand_root, loc=<-51.9729 4.0884 102.6935>, rot=<-0.6084 -0.1793 -0.4613 0.6205>}
 ik_hand_l {parent=ik_hand_gun, loc=<53.0286 82.1782 35.2051>, rot=<-0.5102 -0.7906 -0.3387 0>}
 ik_hand_r {parent=ik_hand_gun, loc=<0 0 0>, rot=<0 0 0 -1>}
]
variables [Crouch Walk {type=bool, default=False}]
