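# Physics asset for the Kurohomura male medium character body, covering two dangling belt chains and the lower body.
# Left and right six-segment box chains use tight swing and twist limits, hanging from base bodies with free rotation on the pelvis.
# Free-rotation cross constraints tie the two chains together at segments 02 and 04.
# Capsules shape pelvis and legs, plus a wide dyn_thighsLookAt box constrained to the right thigh.
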PhysicsAsset /BRCosmetics/Characters/Player/Male/Medium/Bodies/M_MED_Kurohomura/Meshes/M_MED_Kurohomura_Physics
bodies {
 calf_l [{shapeType=Capsule, length=32.962, radius=9.1462, center=<-19.8244 -1.0838 -0.6903>, rotation=<rot 90 0 0>}]
 calf_r [{shapeType=Capsule, length=31.3356, radius=10.4832, center=<19.9602 -0.5088 -0.249>, rotation=<rot 90 0 0>}]
 dyn_belt_01_Base_l [{shapeType=Box, extentX=1.0875, extentY=1.0875, extentZ=0.6548, center=<3 0 0>, rotation=<rot 0 90 0>}]
 dyn_belt_01_Base_r [{shapeType=Box, extentX=1.0875, extentY=1.0875, extentZ=0.6548, center=<3 0 0>, rotation=<rot 0 90 0>}]
 dyn_belt_01_l [{shapeType=Box, extentX=2.2888, extentY=4.5905, extentZ=6.4843, center=<3 0 0>, rotation=<rot 0 90 0>}]
 dyn_belt_01_r [{shapeType=Box, extentX=1.9575, extentY=4.5287, extentZ=6.6231, center=<3 0 0>, rotation=<rot 0 90 0>}]
 dyn_belt_02_l [{shapeType=Box, extentX=2.366, extentY=5.1204, extentZ=6.784, center=<3 0 0>, rotation=<rot 0 90 0>}]
 dyn_belt_02_r [{shapeType=Box, extentX=2.3042, extentY=5.1281, extentZ=6.7424, center=<3 0 0>, rotation=<rot 0 90 0>}]
 dyn_belt_03_l [{shapeType=Box, extentX=2.2708, extentY=5.3926, extentZ=7.4227, center=<3 0 0>, rotation=<rot 0 90 0>}]
 dyn_belt_03_r [{shapeType=Box, extentX=2.2808, extentY=4.9002, extentZ=7.1944, center=<3 0 0>, rotation=<rot 0 90 0>}]
 dyn_belt_04_l [{shapeType=Box, extentX=2.3039, extentY=6.4062, extentZ=7.3802, center=<3 0 0>, rotation=<rot 0 90 0>}]
 dyn_belt_04_r [{shapeType=Box, extentX=2.3692, extentY=6.6137, extentZ=6.5607, center=<3 0 0>, rotation=<rot 0 90 0>}]
 dyn_belt_05_l [{shapeType=Box, extentX=2.6847, extentY=5.6569, extentZ=7.1333, center=<3 0 0>, rotation=<rot 0 90 0>}]
 dyn_belt_05_r [{shapeType=Box, extentX=2.704, extentY=5.7636, extentZ=7.0609, center=<3 0 0>, rotation=<rot 0 90 0>}]
 dyn_belt_06_l [{shapeType=Box, extentX=1.781, extentY=5.8244, extentZ=6.4725, center=<2.2979 -0.0369 0.0916>, rotation=<rot 0 90 0>}]
 dyn_belt_06_r [{shapeType=Box, extentX=2.3037, extentY=5.8244, extentZ=6.4725, center=<2.291 -0.0001 -0.3942>, rotation=<rot 0 89.9999 -7.916>}]
 dyn_thighsLookAt [{shapeType=Box, extentX=33.1451, extentY=52.7715, extentZ=7.529, center=<10.1715 -2.0699 0.9157>, rotation=<rot 0 -90 0>}]
 pelvis [{shapeType=Capsule, length=20.7757, radius=10.8703, center=<-1.6993 2.1279 -0.1048>, rotation=<rot 0 -11.4672 -177.2231>}]
 thigh_l [{shapeType=Capsule, length=33.6895, radius=10.5682, center=<-18.458 2.3178 -0.5595>, rotation=<rot 90 0 -2.4875>}]
 thigh_r [{shapeType=Capsule, length=31.0481, radius=12.2494, center=<20 -0.5103 0.3594>, rotation=<rot 90 0 0>}]
}
constraints [
 {bone1Name=dyn_belt_01_r, bone2Name=dyn_belt_01_Base_r, swing1LimitDegrees=8, swing2LimitDegrees=8, twistLimitDegrees=8}
 {bone1Name=dyn_belt_02_r, bone2Name=dyn_belt_01_r, swing1LimitDegrees=20, swing2LimitDegrees=16, twistLimitDegrees=8}
 {bone1Name=dyn_belt_03_r, bone2Name=dyn_belt_02_r, swing1LimitDegrees=20, swing2LimitDegrees=16, twistLimitDegrees=8}
 {bone1Name=dyn_belt_04_r, bone2Name=dyn_belt_03_r, swing1LimitDegrees=20, swing2LimitDegrees=16, twistLimitDegrees=8}
 {bone1Name=dyn_belt_05_r, bone2Name=dyn_belt_04_r, swing1LimitDegrees=20, swing2LimitDegrees=16, twistLimitDegrees=8}
 {bone1Name=dyn_belt_06_r, bone2Name=dyn_belt_05_r, swing1LimitDegrees=20, swing2LimitDegrees=16, twistLimitDegrees=8}
 {bone1Name=dyn_belt_01_l, bone2Name=dyn_belt_01_Base_l, swing1LimitDegrees=8, swing2LimitDegrees=8, twistLimitDegrees=8}
 {bone1Name=dyn_belt_02_l, bone2Name=dyn_belt_01_l, swing1LimitDegrees=20, swing2LimitDegrees=16, twistLimitDegrees=8}
 {bone1Name=dyn_belt_03_l, bone2Name=dyn_belt_02_l, swing1LimitDegrees=20, swing2LimitDegrees=16, twistLimitDegrees=8}
 {bone1Name=dyn_belt_04_l, bone2Name=dyn_belt_03_l, swing1LimitDegrees=20, swing2LimitDegrees=16, twistLimitDegrees=8}
 {bone1Name=dyn_belt_05_l, bone2Name=dyn_belt_04_l, swing1LimitDegrees=20, swing2LimitDegrees=16, twistLimitDegrees=8}
 {bone1Name=dyn_belt_06_l, bone2Name=dyn_belt_05_l, swing1LimitDegrees=20, swing2LimitDegrees=16, twistLimitDegrees=8}
 {
  bone1Name=dyn_belt_02_l
  bone2Name=dyn_belt_02_r
  swing1LimitDegrees=45
  swing1Motion=Free
  swing2LimitDegrees=45
  swing2Motion=Free
  twistLimitDegrees=45
  twistMotion=Free
 }
 {
  bone1Name=dyn_belt_01_Base_r
  bone2Name=pelvis
  swing1LimitDegrees=45
  swing1Motion=Free
  swing2LimitDegrees=45
  swing2Motion=Free
  twistLimitDegrees=45
  twistMotion=Free
 }
 {
  bone1Name=dyn_belt_01_Base_l
  bone2Name=pelvis
  swing1LimitDegrees=45
  swing1Motion=Free
  swing2LimitDegrees=45
  swing2Motion=Free
  twistLimitDegrees=45
  twistMotion=Free
 }
 {
  bone1Name=thigh_l
  bone2Name=pelvis
  swing1LimitDegrees=45
  swing1Motion=Free
  swing2LimitDegrees=45
  swing2Motion=Free
  twistLimitDegrees=45
  twistMotion=Free
 }
 {
  bone1Name=thigh_r
  bone2Name=pelvis
  swing1LimitDegrees=45
  swing1Motion=Free
  swing2LimitDegrees=45
  swing2Motion=Free
  twistLimitDegrees=45
  twistMotion=Free
 }
 {
  bone1Name=calf_r
  bone2Name=thigh_r
  swing1LimitDegrees=45
  swing1Motion=Free
  swing2LimitDegrees=45
  swing2Motion=Free
  twistLimitDegrees=45
  twistMotion=Free
 }
 {
  bone1Name=calf_l
  bone2Name=thigh_l
  swing1LimitDegrees=45
  swing1Motion=Free
  swing2LimitDegrees=45
  swing2Motion=Free
  twistLimitDegrees=45
  twistMotion=Free
 }
 {
  bone1Name=dyn_belt_04_r
  bone2Name=dyn_belt_04_l
  swing1LimitDegrees=45
  swing1Motion=Free
  swing2LimitDegrees=45
  swing2Motion=Free
  twistLimitDegrees=45
  twistMotion=Free
 }
 {
  bone1Name=dyn_thighsLookAt
  bone2Name=thigh_r
  swing1LimitDegrees=45
  swing1Motion=Free
  swing2LimitDegrees=45
  swing2Motion=Free
  twistLimitDegrees=45
  twistMotion=Free
 }
]
