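# Animation Blueprint for the Kurohomura medium male player body, copying its attached parent mesh's pose and adding procedural secondary motion.
# The copied pose passes through a Control Rig with a crouch-walk input, then base-bone-space RigidBody physics on root.
# Small additive rotations on the left and right belt bones follow.
# Each update, bIsCrouching sets the rig flag and an inverse crouch value driving physics alpha, remapped with a raised minimum.
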AnimBlueprint /BRCosmetics/Characters/Player/Male/Medium/Bodies/M_MED_Kurohomura/Meshes/M_MED_Kurohomura_AnimBP
target_skeleton=&/M_MED_Kurohomura_Skeleton
graphs [
 AnimGraph {nodes [
   n1 = Control Rig!(Source=Copy Pose From Mesh() {bUseAttachedParent=true}, Alpha, Crouch Walk:bool=Get crouch_walk() -> (crouch_walk:bool)) -> (Pose) {
    alphaBoolBlend {blendOption=Linear}
    controlRigAssetReference {blueprintRigClass=&/M_MED_Kurohomura_CtrlRig.M_MED_Kurohomura_CtrlRig_C}
   }
   n2 = RigidBody(ComponentPose=Local To Component(LocalPose=@n1), Alpha=Get crouchAlphaINV() -> (crouchAlphaINV:double)) {
    bClampLinearTranslationLimitToRefPose=true
    bOverrideWorldGravity=true
    overridePhysicsAsset=&/M_MED_Kurohomura_Physics
    simulationSpace=BaseBoneSpace
    alphaBoolBlend {blendOption=Linear}
    alphaScaleBiasClamp {bMapRange=true, outRange {min=0.6}}
    baseBoneRef {boneName=root}
    componentAppliedLinearAccClamp=<1500 1500 1500>
    componentLinearAccScale=<0.4 0.4 0.4>
    componentLinearVelScale=<0.5 0.5 0.5>
    overrideWorldGravity {z=-980}
   }
   n3 = "Transform (Modify) Bone - Bone: dyn_belt_01_l"!(
    ComponentPose=Local To Component(LocalPose=Component To Local(ComponentPose=@n2))
    bAlphaBoolEnabled:bool=True
    Translation="0,0,0"
    Rotation="0,4,0"
    Scale="1,1,1"
   ) -> (Pose) {
    alphaInputType=Bool
    rotationMode=BMM_Additive
    rotationSpace=BCS_BoneSpace
    alphaBoolBlend {blendOption=Linear}
    boneToModify {boneName=dyn_belt_01_l}
    rotation {yaw=4}
   }
   n4 = "Transform (Modify) Bone - Bone: dyn_belt_01_r"!(ComponentPose=@n3, bAlphaBoolEnabled:bool=True, Translation="0,0,0", Rotation="0,4,0", Scale="1,1,1") -> (Pose) {
    alphaInputType=Bool
    rotationMode=BMM_Additive
    rotationSpace=BCS_BoneSpace
    alphaBoolBlend {blendOption=Linear}
    boneToModify {boneName=dyn_belt_01_r}
    rotation {yaw=4}
   }
   Output Pose(Result=Component To Local(ComponentPose=@n4)) {layerGroup=None}
  ]}
 EventGraph {nodes [
   Event BlueprintUpdateAnimation()
   n1 = "To Float (Boolean)"(InBool=Get bIsCrouching())
   > Set crouchAlpha(crouchAlpha:double=@n1) -> (Output_Get:double)
   > Set crouchAlphaINV(crouchAlphaINV:double=float - float(B=@n1) {is=title K2Node_CallFunction}) -> (Output_Get:double)
   > Set crouch_walk(crouch_walk:bool=Get bIsCrouching()) -> (Output_Get:bool)
   TryGetPawnOwner()
  ]}
]
variables [
 crouch_walk {type=boolean}
 crouchAlpha {type=number}
 crouchAlphaINV {type=number}
]
