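# Control Rig for Kurohomura's face accessories that procedurally orients the base of the beard bone chain.
# It resets two aim bones under spine_05, then adds a quarter of the head's relative X Euler rotation to the aim pivot.
# A from-two-vectors quaternion swings the jaw-parented beard base's X axis toward that target.
# A fixed additive rotation then offsets dyn_beard_01_Base; hair bones and the Frontend Alpha variable are unused.
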ControlRigBlueprint /BRCosmetics/Characters/Player/Male/Medium/Bodies/M_MED_Kurohomura/Meshes/Parts/M_MED_Kurohomura_FaceAcc_CtrlRig
graphs [
 RigVMFunctionLibrary {links [], nodes []}
 RigVMModel {nodes [
   n1 = &RigUnit_BeginExecution()
   /Script/ControlRig.RigUnit_SetBoneTransform(
    Bone=dyn_beardBase_Aim
    Space=LocalSpace
    Transform=(Rotation=(X=0,Y=0,Z=0,W=1),Translation=(X=0,Y=0,Z=0),Scale3D=(X=1,Y=1,Z=1))
    Weight=1
    bPropagateToChildren=True
    ExecutePin=@n1.ExecutePin
   ) {is=struct RigVMUnitNode}
   > &RigUnit_SetBoneTransform(
    Bone=dyn_beard_Aim
    Space=LocalSpace
    Transform=(Rotation=(X=0,Y=0,Z=0,W=1),Translation=(X=0,Y=18,Z=0),Scale3D=(X=1,Y=1,Z=1))
    Weight=1
    bPropagateToChildren=True
   )
   n2 = &RigVMFunction_MathQuaternionToEuler(RotationOrder=ZYX, Value=&RigUnit_GetRelativeBoneTransform(Bone=head, Space=spine_05).Transform.Rotation)
   > /Script/ControlRig.RigUnit_ModifyBoneTransforms(
    BoneToModify=((Bone="dyn_beardBase_Aim",Transform=(Rotation=(X=0,Y=0,Z=0,W=1),Translation=(X=0,Y=0,Z=0),Scale3D=(X=1,Y=1,Z=1))))
    Mode=AdditiveLocal
    Weight=1
    WeightMaximum=1
    WeightMinimum=0
    BoneToModify.0.Transform.Rotation=&RigVMFunction_MathQuaternionFromEuler(Euler=/Script/RigVM.RigVMFunction_MathVectorMul(A=(X=0.25,Y=0,Z=0), B=@n2.Result) {is=struct RigVMUnitNode, template="Multiply::Execute(in A,in B,out Result)"}.Result).Result
   ) {is=struct RigVMUnitNode}
   n3 = &RigUnit_GetBoneTransform(Bone=dyn_beardBaseJaw)
   n4 = /Script/RigVM.RigVMFunction_MathQuaternionGetAxis(Axis=X, Quaternion=@n3.Transform.Rotation) {is=struct RigVMUnitNode, template="Axis::Execute(in Quaternion,in Axis,out Result)"}
   n5 = /Script/RigVM.RigVMFunction_MathVectorSub(
    B=@n3.Transform.Translation
    A=/Script/RigVM.RigVMFunction_MathTransformTransformVector(Location=(X=0,Y=1,Z=0), Transform=&RigUnit_GetBoneTransform(Bone=dyn_beard_Aim).Transform) {is=struct RigVMUnitNode}.Result
   ) {is=struct RigVMUnitNode, template="Subtract::Execute(in A,in B,out Result)"}
   > /Script/ControlRig.RigUnit_SetBoneRotation(
    Bone=dyn_beardBaseJaw
    Space=GlobalSpace
    Weight=1
    bPropagateToChildren=True
    Rotation=/Script/RigVM.RigVMFunction_MathQuaternionMul(
     B=@n3.Transform.Rotation
     A=/Script/RigVM.RigVMFunction_MathQuaternionFromTwoVectors(A=@n4.Result, B=@n5.Result) {is=struct RigVMUnitNode, template="FromTwoVectors::Execute(in A,in B,out Result)"}.Result
    ) {is=struct RigVMUnitNode, template="Multiply::Execute(in A,in B,out Result)"}.Result
   ) {is=struct RigVMUnitNode}
   > &RigUnit_ModifyBoneTransforms(
    BoneToModify=((Bone="dyn_beard_01_Base",Transform=(Rotation=(X=-0.1236,Y=0.1349,Z=-0.0259,W=0.9828),Translation=(X=0,Y=0,Z=0),Scale3D=(X=1,Y=1,Z=1))),(Bone="None",Transform=(Rotation=(X=0,Y=0,Z=0,W=1),Translation=(X=0,Y=0,Z=0),Scale3D=(X=1,Y=1,Z=1))))
    Mode=AdditiveLocal
    Weight=1
    WeightMaximum=1
    WeightMinimum=0
   )
  ]}
]
hierarchy [
 root {loc=<0 0 0>, rot=<0 0 0 1>}
 attach {parent=root, loc=<0 0 0>, rot=<0 0 0 1>}
 pelvis {parent=root, loc=<0 -0.7071 93.1191>, rot=<0.0127 0.707 -0.0127 -0.707>}
 spine_01 {parent=pelvis, loc=<6.7494 -0.6388 0>, rot=<0 0 0.034 -0.9994>}
 spine_02 {parent=spine_01, loc=<7.2518 -0.035 0>, rot=<0 0 0.0238 -0.9997>}
 spine_03 {parent=spine_02, loc=<6.0104 0 0>, rot=<0 0 0.0149 -0.9999>}
 spine_04 {parent=spine_03, loc=<8.1024 0 0>, rot=<0 0 -0.0132 -0.9999>}
 spine_05 {parent=spine_04, loc=<10.0146 0 0>, rot=<0 0 -0.0015 -1>}
 clavicle_l {parent=spine_05, loc=<4.3575 -0.1466 -2.3223>, rot=<0.0465 -0.7289 -0.009 -0.6829>}
 clavicle_r {parent=spine_05, loc=<4.3575 -0.1466 2.3223>, rot=<0.7289 0.0465 -0.6829 0.009>}
 neck_01 {parent=spine_05, loc=<11.3138 0.1529 0>, rot=<0 0 -0.2003 -0.9797>}
 neck_02 {parent=neck_01, loc=<4.5627 0 0>, rot=<0 0 0.0504 -0.9987>}
 head {parent=neck_02, loc=<5.3914 0 0>, rot=<0 0 0.115 -0.9934>}
 faceAttach {parent=head, loc=<-4.0871 0.3683 0.025>, rot=<0.0082 -0.7071 -0.0082 -0.7071>, scale=<0.89 0.89 0.89>}
 C_brow_mid {parent=faceAttach, loc=<0.0373 10.2314 10.7339>, rot=<0 0.0098 -0.0019 -0.9999>}
 L_brow_mid {parent=faceAttach, loc=<3.2396 10.1391 11.3046>, rot=<0 -0.0032 0.2553 -0.9669>}
 L_brow_outer {parent=faceAttach, loc=<6.1212 8.5405 11.4681>, rot=<0.0179 0.0186 0.4385 -0.8984>}
 L_cheek_inner {parent=faceAttach, loc=<5.602 8.1502 6.0929>, rot=<-0.0769 -0.2855 0.3161 -0.9015>}
 C_lip_upper_mid {parent=faceAttach, loc=<0.0327 10.5657 4.0562>, rot=<0.2225 0.0006 -0.0002 -0.9749>}
 L_lip_upper_outer {parent=faceAttach, loc=<2.5867 10.0828 4.314>, rot=<0.0375 -0.0684 0.2611 -0.9622>}
 L_lip_corner {parent=faceAttach, loc=<3.3388 9.1967 3.6383>, rot=<0.0035 0.0074 -0.0147 -0.9999>}
 C_nose_bridge {parent=faceAttach, loc=<-0.0542 11.3158 7.8025>, rot=<-0.2218 0.0095 0.0027 -0.975>}
 R_brow_mid {parent=faceAttach, loc=<-3.1127 10.1513 11.2967>, rot=<-0.0001 0.0042 -0.2553 -0.9669>}
 R_brow_outer {parent=faceAttach, loc=<-6.086 8.5021 11.4894>, rot=<-0.0087 0.0149 -0.4449 -0.8954>}
 R_cheek_inner {parent=faceAttach, loc=<-5.5252 8.1222 6.0944>, rot=<-0.0781 0.2886 -0.3156 -0.9006>}
 R_lip_corner {parent=faceAttach, loc=<-3.3134 9.135 3.7009>, rot=<0.003 -0.008 0.0146 -0.9999>}
 C_jaw {parent=faceAttach, loc=<-0.0551 2.9908 5.0746>, rot=<0 0 0 1>}
 L_lip_lower_outer {parent=C_jaw, loc=<2.7437 7.1132 -1.1923>, rot=<-0.002 -0.0366 0.2428 -0.9694>}
 C_lip_lower_mid {parent=C_jaw, loc=<0.1328 7.5274 -1.2092>, rot=<0 -0.058 -0.0123 -0.9982>}
 R_lip_lower_outer {parent=C_jaw, loc=<-2.5545 7.091 -1.1875>, rot=<-0.0103 0.0818 -0.2065 -0.975>}
 teeth_lower {parent=C_jaw, loc=<-0.0503 3.9644 -1.9834>, rot=<0 0 0 1>}
 tongue {parent=C_jaw, loc=<0.0575 1.6759 -1.2107>, rot=<-0.1568 0 0 -0.9876>}
 dyn_beardBaseJaw {parent=C_jaw, loc=<0.0629 6.729 -7.8201>, rot=<0.6807 0.1905 -0.673 -0.2175>}
 dyn_beard_01_Base {parent=dyn_beardBaseJaw, loc=<3.3554 0.0001 0>, rot=<0.0533 0.0196 -0.1123 -0.9921>}
 dyn_beard_01 {parent=dyn_beard_01_Base, loc=<0 0 0>, rot=<0 0 0 1>}
 dyn_beard_02 {parent=dyn_beard_01, loc=<8.6204 0 0>, rot=<0.0002 0 0.0262 -0.9997>}
 dyn_beard_03 {parent=dyn_beard_02, loc=<9 0 0>, rot=<-0.0579 -0.0247 0.0189 -0.9978>}
 dyn_beard_04 {parent=dyn_beard_03, loc=<8.5 0 0>, rot=<0.0288 0.0118 -0.0325 -0.999>}
 dyn_beard_05 {parent=dyn_beard_04, loc=<6.0761 0 0>, rot=<-0.0004 -0.0002 0.0262 -0.9997>}
 dyn_beard_06 {parent=dyn_beard_05, loc=<7.3218 0 0>, rot=<0.0559 0.0294 0.0213 -0.9978>}
 dyn_beard_07 {parent=dyn_beard_06, loc=<7.5 0 0>, rot=<0 0 0 1>}
 R_eye {parent=faceAttach, loc=<-3.0847 7.6365 9.9853>, rot=<0 0 0.9998 -0.0187>}
 R_eye_lid_upper_mid {parent=R_eye, loc=<-0.0904 -1.1707 0.2535>, rot=<0.0541 0.1725 0.0088 -0.9835>, scale=<1 1 1.0001>}
 R_eye_lid_lower_mid {parent=R_eye, loc=<0.2504 -0.9477 -0.6517>, rot=<0.0351 0.173 0.0116 -0.9842>}
 L_eye {parent=faceAttach, loc=<3.1408 7.6365 9.9853>, rot=<0 0 0.9998 0.0187>}
 L_eye_lid_upper_mid {parent=L_eye, loc=<0.0633 -1.164 0.2233>, rot=<0.0515 -0.1734 0.0091 -0.9835>}
 L_eye_lid_lower_mid {parent=L_eye, loc=<-0.2554 -0.9203 -0.7078>, rot=<0.0324 -0.1736 0.0057 -0.9843>}
 teeth_upper {parent=faceAttach, loc=<0.0105 6.6187 4.1121>, rot=<0 0 0 1>}
 R_lip_upper_outer {parent=faceAttach, loc=<-2.5428 10.103 4.3113>, rot=<0.0449 0.0476 -0.2531 -0.9652>}
 transform1 {parent=faceAttach, loc=<0.028 3.9037 -166.1968>, rot=<0.0074 0 0 1>, scale=<1.1236 1.1236 1.1236>}
 dyn_hairMain_01_Base {parent=head, loc=<9.0857 -6.6694 0.0739>, rot=<0.0099 -0.0143 0.5612 -0.8275>, scale=<0.9 0.9 0.9>}
 dyn_hairMain_01 {parent=dyn_hairMain_01_Base, loc=<0 0 0>, rot=<0 0 0 1>}
 dyn_hairMain_02 {parent=dyn_hairMain_01, loc=<15.352 0 0>, rot=<0.0536 -0.02 0.0126 -0.9983>}
 dyn_hairLockG_01 {parent=dyn_hairMain_02, loc=<4.652 -0.8892 2.9848>, rot=<0.8515 0.0049 -0.0494 -0.5219>}
 dyn_hairLockH_01 {parent=dyn_hairMain_02, loc=<8.0567 -6.373 -1.0573>, rot=<-0.0245 0.0443 -0.2339 -0.9709>}
 dyn_hairLockJ_01 {parent=dyn_hairMain_02, loc=<4.5716 -7.1844 -6.2922>, rot=<-0.3613 0.0065 -0.0974 -0.9273>}
 dyn_hairLockN_01 {parent=dyn_hairMain_02, loc=<4.9301 2.9029 -7.3783>, rot=<-0.7839 0.2172 -0.0988 -0.5731>}
 dyn_hairLockS_01 {parent=dyn_hairMain_02, loc=<5.5695 6.064 -0.6848>, rot=<-0.658 0.2075 0.029 -0.7233>}
 dyn_hairLockA_01 {parent=dyn_hairMain_01, loc=<2.7445 4.9951 3.5305>, rot=<0.9073 0.4027 0.0188 -0.1198>}
 dyn_hairLockA_02 {parent=dyn_hairLockA_01, loc=<3.8417 0 0>, rot=<-0.0899 0.1447 -0.1824 -0.9684>}
 dyn_hairLockB_01 {parent=dyn_hairMain_01, loc=<5.2698 3.5893 5.5868>, rot=<0.9287 0.2957 -0.0174 -0.223>}
 dyn_hairLockB_02 {parent=dyn_hairLockB_01, loc=<5.6343 0 0>, rot=<-0.1245 0.0347 -0.1656 -0.9777>}
 dyn_hairLockC_01 {parent=dyn_hairMain_01, loc=<13.7288 5.6708 8.5557>, rot=<0.8861 0.1131 0.008 -0.4494>}
 dyn_hairLockD_01 {parent=dyn_hairMain_01, loc=<13.4923 0.1445 9.5083>, rot=<0.7525 0.0417 -0.035 -0.6564>}
 dyn_hairLockE_01 {parent=dyn_hairMain_01, loc=<9.1464 -0.1148 10.6484>, rot=<0.6308 0.109 0.0875 -0.7633>}
 dyn_hairLockF_01 {parent=dyn_hairMain_01, loc=<17.3241 -5.6381 9.2123>, rot=<0.423 -0.0325 -0.1636 -0.8906>}
 dyn_hairLockK_01 {parent=dyn_hairMain_01, loc=<9.6845 -9.9257 -5.3384>, rot=<-0.3167 0.0684 0.5556 -0.7657>}
 dyn_hairLockL_01 {parent=dyn_hairMain_01, loc=<1.272 -4.9848 -3.6957>, rot=<-0.5128 0.0174 0.462 -0.7234>}
 dyn_hairLockM_01 {parent=dyn_hairMain_01, loc=<16.8047 -3.1424 -9.4053>, rot=<-0.6743 -0.0437 -0.095 -0.731>}
 dyn_hairLockP_01 {parent=dyn_hairMain_01, loc=<3.9758 3.4999 -3.9825>, rot=<0.8671 0.3153 -0.1337 0.3617>}
 dyn_hairLockR_01 {parent=dyn_hairMain_01, loc=<10.4457 7.6769 -3.8034>, rot=<0.9159 0.0361 0.0019 0.3998>}
 dyn_sideburns_Base_l {parent=head, loc=<0.2513 4.1734 -6.4359>, rot=<-0.4013 0.4431 0.7882 0.1465>}
 dyn_sideburns_l {parent=dyn_sideburns_Base_l, loc=<0 0 0>, rot=<0 0 0 1>}
 dyn_sideburns_Base_r {parent=head, loc=<0.3913 4.1862 6.7477>, rot=<0.3913 -0.3479 0.8479 0.0836>}
 dyn_sideburns_r {parent=dyn_sideburns_Base_r, loc=<0 0 0>, rot=<0 0 0 1>}
 dyn_alfalfa {parent=head, loc=<10.755 9.7138 2.365>, rot=<0.9384 -0.118 0.1317 -0.2968>}
 dyn_beardBase_Aim {parent=spine_05, loc=<-131.0232 -6.9525 0>, rot=<0 0 0 1>}
 dyn_beard_Aim {parent=dyn_beardBase_Aim, loc=<131.0232 6.953 0>, rot=<0 0 0 1>}
]
variables [Get Frontend Alpha {type=double, default=0}]
