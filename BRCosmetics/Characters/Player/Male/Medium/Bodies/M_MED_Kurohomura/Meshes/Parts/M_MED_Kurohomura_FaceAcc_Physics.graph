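# Physics asset for the Kurohomura male medium body's face accessory mesh, defining secondary motion for beard, hair and sideburns.
# A seven-segment beard chain hangs from a jaw base, while sideburns, an alfalfa tuft and a two-segment main hair chain with branching locks attach to the head.
# Dynamic segments have tight angular limits; their base bodies and the head-and-spine chain, which supplies torso volumes, use free angular constraints.
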PhysicsAsset /BRCosmetics/Characters/Player/Male/Medium/Bodies/M_MED_Kurohomura/Meshes/Parts/M_MED_Kurohomura_FaceAcc_Physics
bodies {
 dyn_alfalfa [{shapeType=Box, extentX=1.5621, extentY=2.478, extentZ=7.6628, center=<2.8596 0 -0.0001>, rotation=<rot 90 0 0>}]
 dyn_beardBaseJaw [{shapeType=Capsule, length=3.9129, radius=2.6245, center=<2.0086 -0.6176 -0.3674>, rotation=<rot 1.5006 -90 4.4291>}]
 dyn_beard_01 [{shapeType=Box, extentX=3.8757, extentY=9.8765, extentZ=11.6211, center=<4.2355 0.0707 0.1796>, rotation=<rot 0 86.5679 -88.9597>}]
 dyn_beard_01_Base [{shapeType=Capsule, length=0.505, radius=0.505, rotation=<rot 0 -90 0>}]
 dyn_beard_02 [{shapeType=Box, extentX=3.6839, extentY=9.7453, extentZ=12.0709, center=<4.3868 -0.2713 -0.1696>, rotation=<rot 0 87.2228 -87.1917>}]
 dyn_beard_03 [{shapeType=Box, extentX=3.7581, extentY=8.3152, extentZ=11.6592, center=<4.2109 -0.064 -0.2358>, rotation=<rot 0 89.324 -87.5549>}]
 dyn_beard_04 [{shapeType=Capsule, length=5.7889, radius=3.0366, center=<2.941 -0.151 -0.058>, rotation=<rot 0 87.5894 -87.8971>}]
 dyn_beard_05 [{shapeType=Capsule, length=4.8086, radius=2.9349, center=<3.6633 -0.1466 -0.0269>, rotation=<rot 0 -91.0677 -89.0674>}]
 dyn_beard_06 [{shapeType=Capsule, length=6.0219, radius=2.6984, center=<3.578 0.0603 -0.0757>, rotation=<rot 0 87.0946 -87.152>}]
 dyn_beard_07 [{shapeType=Capsule, length=7.3903, radius=3.9695, center=<5.5949 -0.6552 0.8314>, rotation=<rot -0.0001 78.9338 -82.5507>}]
 dyn_hairLockA_01 [{shapeType=Capsule, length=3.2439, radius=0.8491, center=<2 0 0>, rotation=<rot 90 0 0>}]
 dyn_hairLockA_02 [{shapeType=Capsule, length=3.2056, radius=1.7016, center=<3 0 0>, rotation=<rot 90 0 0>}]
 dyn_hairLockB_01 [{shapeType=Capsule, length=4.6157, radius=1.5123, center=<2 0 0>, rotation=<rot 90 0 0>}]
 dyn_hairLockB_02 [{shapeType=Capsule, length=3.464, radius=1.6805, center=<3 0 0>, rotation=<rot 90 0 0>}]
 dyn_hairLockC_01 [{shapeType=Capsule, length=4.1942, radius=1.6008, center=<5.8482 -0.1993 0.1414>, rotation=<rot 0 -87.7788 -89.2153>}]
 dyn_hairLockD_01 [{shapeType=Capsule, length=5.9149, radius=3.5843, center=<7.1388 -0.1262 0.1115>, rotation=<rot 0 90.8315 -86.1123>}]
 dyn_hairLockE_01 [{shapeType=Capsule, length=5.154, radius=1.6526, center=<4.6247 -0.1961 -0.0181>, rotation=<rot 0 -90.7719 -82.1609>}]
 dyn_hairLockF_01 [{shapeType=Capsule, length=5.5586, radius=2.2986, center=<5.4976 -0.6152 -0.1383>, rotation=<rot 0 97.5579 -89.659>}]
 dyn_hairLockG_01 [{shapeType=Capsule, length=8.3879, radius=2.8021, center=<4.1597 0 0>, rotation=<rot 90 0 0>}]
 dyn_hairLockH_01 [{shapeType=Capsule, length=3.4017, radius=2.0493, center=<3.5178 0 0>, rotation=<rot 90 0 0>}]
 dyn_hairLockJ_01 [{shapeType=Capsule, length=4.3493, radius=2.7416, center=<3 0 0>, rotation=<rot 90 0 0>}]
 dyn_hairLockK_01 [{shapeType=Capsule, length=2.8426, radius=2.4909, center=<1.2507 0.8207 -1.1524>, rotation=<rot 0 -102.0618 -50.2063>}]
 dyn_hairLockL_01 [{shapeType=Capsule, length=4.3466, radius=2.3468, center=<5.3838 -0.4965 -0.0865>, rotation=<rot 0 101.1148 -84.2289>}]
 dyn_hairLockM_01 [{shapeType=Capsule, length=6.2261, radius=2.3157, center=<4.3574 -0.64 -0.0057>, rotation=<rot 0 -85.0383 -85.8739>}]
 dyn_hairLockN_01 [{shapeType=Capsule, length=9.8062, radius=2.3112, center=<6.5438 0.4118 0.1654>, rotation=<rot 90 0 0>}]
 dyn_hairLockP_01 [{shapeType=Capsule, length=5.9643, radius=1.7102, center=<4.1187 -0.6818 0.3769>, rotation=<rot 0 -93.4659 -78.0469>}]
 dyn_hairLockR_01 [{shapeType=Capsule, length=4.1618, radius=1.7357, center=<5.8196 -0.338 0.2867>, rotation=<rot 0 97.6206 -84.8546>}]
 dyn_hairLockS_01 [{shapeType=Capsule, length=11.0106, radius=2.753, center=<5.583 0 -0.0001>, rotation=<rot 90 0 0>}]
 dyn_hairMain_01 [{shapeType=Capsule, length=11.5642, radius=10.8046, center=<7.8841 -0.8121 0.22>, rotation=<rot 90 0 0>}]
 dyn_hairMain_01_Base [{shapeType=Capsule, length=0.505, radius=0.505, rotation=<rot 0 -90 0>}]
 dyn_hairMain_02 [{shapeType=Capsule, length=5.9313, radius=8.4489, center=<5.092 -0.1485 0.5243>, rotation=<rot 90 0 0>}]
 dyn_sideburns_Base_l [{shapeType=Box, extentX=1.01, extentY=1.01, extentZ=1.01, rotation=<rot 0 -90 0>}]
 dyn_sideburns_Base_r [{shapeType=Box, extentX=1.01, extentY=1.01, extentZ=1.01, rotation=<rot 0 -90 0>}]
 dyn_sideburns_l [{shapeType=Box, extentX=3.9755, extentY=6.7279, extentZ=8.4193, center=<2.8208 0.2851 -0.3922>, rotation=<rot 13.316 -90 6.3052>}]
 dyn_sideburns_r [{shapeType=Box, extentX=3.863, extentY=6.6217, extentZ=7.5623, center=<2.0566 0.5141 0.3298>, rotation=<rot -0.9136 -90 -21.8765>}]
 head [{shapeType=Capsule, length=8.949, radius=8.8573, center=<2.5996 3.7471 0.0121>, rotation=<rot 0 59.3977 -88.342>}]
 spine_01 [
  {shapeType=Sphere, radius=5.9025, center=<-4.0872 13.0148 0>}
  {shapeType=Capsule, length=14.6752, radius=11.5014, center=<0 2.3354 0>, rotation=<rot 0 -90 0>}
 ]
 spine_02 [{shapeType=Capsule, length=14.5071, radius=11.6695, center=<0 2.3109 0.0002>, rotation=<rot 0 -90 0>}]
 spine_03 [{shapeType=Capsule, length=13.9931, radius=12.1836, center=<0 2.6068 0>, rotation=<rot 0 -90 0>}]
 spine_04 [{shapeType=Capsule, length=14.5646, radius=12.612, center=<0 3.2718 0.0001>, rotation=<rot 0 -90 0>}]
 spine_05 [{shapeType=Capsule, length=14.6752, radius=12.5014, center=<-0.3065 0.3037 0>, rotation=<rot 0 -90 0>}]
}
constraints [
 {
  bone1Name=dyn_beardBaseJaw
  bone2Name=head
  swing1LimitDegrees=45
  swing1Motion=Free
  swing2LimitDegrees=45
  swing2Motion=Free
  twistLimitDegrees=45
  twistMotion=Free
 }
 {
  bone1Name=dyn_beard_01_Base
  bone2Name=dyn_beardBaseJaw
  swing1LimitDegrees=45
  swing1Motion=Free
  swing2LimitDegrees=45
  swing2Motion=Free
  twistLimitDegrees=45
  twistMotion=Free
 }
 {bone1Name=dyn_beard_01, bone2Name=dyn_beard_01_Base, swing1LimitDegrees=18, swing2LimitDegrees=18, twistLimitDegrees=8}
 {bone1Name=dyn_beard_02, bone2Name=dyn_beard_01, swing1LimitDegrees=18, swing2LimitDegrees=18, twistLimitDegrees=8}
 {bone1Name=dyn_beard_03, bone2Name=dyn_beard_02, swing1LimitDegrees=18, swing2LimitDegrees=18, twistLimitDegrees=8}
 {bone1Name=dyn_beard_04, bone2Name=dyn_beard_03, swing1LimitDegrees=18, swing2LimitDegrees=18, twistLimitDegrees=8}
 {bone1Name=dyn_beard_05, bone2Name=dyn_beard_04, swing1LimitDegrees=18, swing2LimitDegrees=18, twistLimitDegrees=8}
 {bone1Name=dyn_beard_06, bone2Name=dyn_beard_05, swing1LimitDegrees=18, swing2LimitDegrees=18, twistLimitDegrees=8}
 {bone1Name=dyn_beard_07, bone2Name=dyn_beard_06, swing1LimitDegrees=18, swing2LimitDegrees=18, twistLimitDegrees=8}
 {
  bone1Name=head
  bone2Name=spine_05
  swing1LimitDegrees=45
  swing1Motion=Free
  swing2LimitDegrees=45
  swing2Motion=Free
  twistLimitDegrees=45
  twistMotion=Free
 }
 {
  bone1Name=spine_02
  bone2Name=spine_01
  swing1LimitDegrees=45
  swing1Motion=Free
  swing2LimitDegrees=45
  swing2Motion=Free
  twistLimitDegrees=45
  twistMotion=Free
 }
 {
  bone1Name=spine_03
  bone2Name=spine_02
  swing1LimitDegrees=45
  swing1Motion=Free
  swing2LimitDegrees=45
  swing2Motion=Free
  twistLimitDegrees=45
  twistMotion=Free
 }
 {
  bone1Name=spine_04
  bone2Name=spine_03
  swing1LimitDegrees=45
  swing1Motion=Free
  swing2LimitDegrees=45
  swing2Motion=Free
  twistLimitDegrees=45
  twistMotion=Free
 }
 {
  bone1Name=spine_05
  bone2Name=spine_04
  swing1LimitDegrees=45
  swing1Motion=Free
  swing2LimitDegrees=45
  swing2Motion=Free
  twistLimitDegrees=45
  twistMotion=Free
 }
 {bone1Name=dyn_alfalfa, bone2Name=head, swing1LimitDegrees=20, swing2LimitDegrees=12, twistLimitDegrees=6}
 {bone1Name=dyn_sideburns_r, bone2Name=dyn_sideburns_Base_r, swing1LimitDegrees=16, swing2LimitDegrees=4, twistLimitDegrees=4}
 {
  bone1Name=dyn_sideburns_Base_r
  bone2Name=head
  swing1LimitDegrees=45
  swing1Motion=Free
  swing2LimitDegrees=45
  swing2Motion=Free
  twistLimitDegrees=45
  twistMotion=Free
 }
 {bone1Name=dyn_sideburns_l, bone2Name=dyn_sideburns_Base_l, swing1LimitDegrees=16, swing2LimitDegrees=4, twistLimitDegrees=4}
 {
  bone1Name=dyn_sideburns_Base_l
  bone2Name=head
  swing1LimitDegrees=45
  swing1Motion=Free
  swing2LimitDegrees=45
  swing2Motion=Free
  twistLimitDegrees=45
  twistMotion=Free
 }
 {
  bone1Name=dyn_hairMain_01_Base
  bone2Name=head
  swing1LimitDegrees=45
  swing1Motion=Free
  swing2LimitDegrees=45
  swing2Motion=Free
  twistLimitDegrees=45
  twistMotion=Free
 }
 {bone1Name=dyn_hairMain_01, bone2Name=dyn_hairMain_01_Base, swing1LimitDegrees=12, swing2LimitDegrees=12, twistLimitDegrees=12}
 {bone1Name=dyn_hairMain_02, bone2Name=dyn_hairMain_01, swing1LimitDegrees=20, swing2LimitDegrees=20, twistLimitDegrees=20}
 {bone1Name=dyn_hairLockG_01, bone2Name=dyn_hairMain_02, swing1LimitDegrees=16, swing2LimitDegrees=16, twistLimitDegrees=10}
 {bone1Name=dyn_hairLockH_01, bone2Name=dyn_hairMain_02, swing1LimitDegrees=16, swing2LimitDegrees=16, twistLimitDegrees=10}
 {bone1Name=dyn_hairLockJ_01, bone2Name=dyn_hairMain_02, swing1LimitDegrees=16, swing2LimitDegrees=16, twistLimitDegrees=10}
 {bone1Name=dyn_hairLockN_01, bone2Name=dyn_hairMain_02, swing1LimitDegrees=16, swing2LimitDegrees=16, twistLimitDegrees=10}
 {bone1Name=dyn_hairLockS_01, bone2Name=dyn_hairMain_02, swing1LimitDegrees=16, swing2LimitDegrees=16, twistLimitDegrees=10}
 {bone1Name=dyn_hairLockA_01, bone2Name=dyn_hairMain_01, swing1LimitDegrees=4, swing2LimitDegrees=4, twistLimitDegrees=4}
 {bone1Name=dyn_hairLockA_02, bone2Name=dyn_hairLockA_01, swing1LimitDegrees=12, swing2LimitDegrees=12, twistLimitDegrees=8}
 {bone1Name=dyn_hairLockB_01, bone2Name=dyn_hairMain_01, swing1LimitDegrees=4, swing2LimitDegrees=4, twistLimitDegrees=4}
 {bone1Name=dyn_hairLockB_02, bone2Name=dyn_hairLockB_01, swing1LimitDegrees=12, swing2LimitDegrees=12, twistLimitDegrees=8}
 {bone1Name=dyn_hairLockC_01, bone2Name=dyn_hairMain_01, swing1LimitDegrees=16, swing2LimitDegrees=16, twistLimitDegrees=12}
 {bone1Name=dyn_hairLockD_01, bone2Name=dyn_hairMain_01, swing1LimitDegrees=16, swing2LimitDegrees=16, twistLimitDegrees=12}
 {bone1Name=dyn_hairLockE_01, bone2Name=dyn_hairMain_01, swing1LimitDegrees=16, swing2LimitDegrees=16, twistLimitDegrees=12}
 {bone1Name=dyn_hairLockF_01, bone2Name=dyn_hairMain_01, swing1LimitDegrees=16, swing2LimitDegrees=16, twistLimitDegrees=12}
 {bone1Name=dyn_hairLockK_01, bone2Name=dyn_hairMain_01, swing1LimitDegrees=16, swing2LimitDegrees=16, twistLimitDegrees=12}
 {bone1Name=dyn_hairLockL_01, bone2Name=dyn_hairMain_01, swing1LimitDegrees=16, swing2LimitDegrees=16, twistLimitDegrees=12}
 {bone1Name=dyn_hairLockM_01, bone2Name=dyn_hairMain_01, swing1LimitDegrees=16, swing2LimitDegrees=16, twistLimitDegrees=12}
 {bone1Name=dyn_hairLockP_01, bone2Name=dyn_hairMain_01, swing1LimitDegrees=16, swing2LimitDegrees=16, twistLimitDegrees=12}
 {bone1Name=dyn_hairLockR_01, bone2Name=dyn_hairMain_01, swing1LimitDegrees=16, swing2LimitDegrees=16, twistLimitDegrees=12}
]
